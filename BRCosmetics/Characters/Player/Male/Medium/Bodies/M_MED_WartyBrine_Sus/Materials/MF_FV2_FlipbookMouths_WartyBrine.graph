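# Material function for the WartyBrine character that overlays an animated flipbook mouth onto incoming material attributes.
# A Switch remaps flipbook_face_mouth_index to frames of a 4x4 SubUV texture, sampled on UV0 or UV1 with adjustable scale and offset.
# Texture alpha times two per-axis SphereMasks confines the blend into BaseColor and, via static switch, EmissiveColor.
# It outputs the modified attributes plus mouth Color and Alpha.
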MaterialFunction /BRCosmetics/Characters/Player/Male/Medium/Bodies/M_MED_WartyBrine_Sus/Materials/MF_FV2_FlipbookMouths_WartyBrine
graph {expressions {
  MaterialExpressionAdd_0 {
   class=MaterialExpressionAdd
   inputs [
    A {from=MaterialExpressionMultiply_1}
    B {from=MaterialExpressionAppendVector_0}
   ]
  }
  MaterialExpressionAppendVector_0 {
   class=MaterialExpressionAppendVector
   inputs [
    A {from=MaterialExpressionScalarParameter_16, output=""}
    B {from=MaterialExpressionScalarParameter_17, output=""}
   ]
  }
  MaterialExpressionAppendVector_1 {
   class=MaterialExpressionAppendVector
   inputs [
    A {from=MaterialExpressionConstant_3, output=""}
    B {from=MaterialExpressionScalarParameter_2, output=""}
   ]
  }
  MaterialExpressionConstant3Vector_0 {class=MaterialExpressionConstant3Vector, properties {constant=<rgba 13 14 15 0>}}
  MaterialExpressionConstant3Vector_1 {class=MaterialExpressionConstant3Vector, properties {constant=<rgba 1 5 6 0>}}
  MaterialExpressionConstant4Vector_0 {class=MaterialExpressionConstant4Vector, inputs [], properties {constant=<rgba 9 0 10 11>}}
  MaterialExpressionConstant_0 {class=MaterialExpressionConstant}
  MaterialExpressionConstant_1 {class=MaterialExpressionConstant, properties {r=4}}
  MaterialExpressionConstant_11 {class=MaterialExpressionConstant, properties {r=0.5}}
  MaterialExpressionConstant_2 {class=MaterialExpressionConstant, properties {r=4}}
  MaterialExpressionConstant_3 {class=MaterialExpressionConstant, properties {r=1}}
  MaterialExpressionFunctionInput_18 {
   class=MaterialExpressionFunctionInput
   inputs [Preview {from=MaterialExpressionAdd_0, output=""}]
   properties {bUsePreviewValueAsDefault=true, inputName=MouthUVs, inputType=FunctionInput_Vector2}
  }
  MaterialExpressionFunctionInput_19 {
   class=MaterialExpressionFunctionInput
   inputs [Preview {from="", output=""}]
   properties {bUsePreviewValueAsDefault=true, inputName=Material, inputType=FunctionInput_MaterialAttributes}
  }
  MaterialExpressionFunctionOutput_0 {class=MaterialExpressionFunctionOutput, inputs [None {from=MaterialExpressionSetMaterialAttributes_0, output=""}]}
  MaterialExpressionFunctionOutput_4 {
   class=MaterialExpressionFunctionOutput
   inputs [None {from=MaterialExpressionMaterialFunctionCall_2, output=RGB}]
   properties {outputName=Color}
  }
  MaterialExpressionFunctionOutput_5 {class=MaterialExpressionFunctionOutput, inputs [None {from=MaterialExpressionMultiply_0, output=""}], properties {outputName=Alpha}}
  MaterialExpressionGetMaterialAttributes_0 {
   class=MaterialExpressionGetMaterialAttributes
   inputs [None {from=MaterialExpressionFunctionInput_19, output=""}]
   properties {attributeGetTypes [#1, #2]}
  }
  MaterialExpressionLinearInterpolate_0 {
   class=MaterialExpressionLinearInterpolate
   inputs [
    A {from=MaterialExpressionGetMaterialAttributes_0, output=EmissiveColor}
    B {from=MaterialExpressionMultiply_2}
    Alpha {from=MaterialExpressionMultiply_0}
   ]
   properties {constB=0}
  }
  MaterialExpressionLinearInterpolate_15 {
   class=MaterialExpressionLinearInterpolate
   inputs [
    A {from=MaterialExpressionGetMaterialAttributes_0, output=BaseColor}
    B {from=MaterialExpressionReroute_0}
    Alpha {from=MaterialExpressionMultiply_0}
   ]
  }
  MaterialExpressionMaterialFunctionCall_2 {
   class=MaterialExpressionMaterialFunctionCall
   inputs [
    Texture {from=MaterialExpressionTextureObjectParameter_0}
    UVs {from=MaterialExpressionFunctionInput_18}
    SubImages {from=MaterialExpressionConstant_2}
    Frame {from=MaterialExpressionSwitch_0}
   ]
   properties {materialFunction=/Engine/Functions/Engine_MaterialFunctions03/Texturing/SubUV_Function}
  }
  MaterialExpressionMaterialFunctionCall_36 {
   class=MaterialExpressionMaterialFunctionCall
   inputs [Float2 {from=MaterialExpressionMaterialFunctionCall_2, output=UVs}]
   properties {materialFunction=&BreakOutFloat2Components}
  }
  MaterialExpressionMultiply_0 {
   class=MaterialExpressionMultiply
   inputs [
    A {from=MaterialExpressionMaterialFunctionCall_2, output=Alpha}
    B {from=MaterialExpressionMultiply_13}
   ]
  }
  MaterialExpressionMultiply_1 {
   class=MaterialExpressionMultiply
   inputs [
    A {from=MaterialExpressionStaticSwitchParameter_0}
    B {from=MaterialExpressionMultiply_3}
   ]
  }
  MaterialExpressionMultiply_13 {
   class=MaterialExpressionMultiply
   inputs [
    A {from=MaterialExpressionSphereMask_30}
    B {from=MaterialExpressionSphereMask_29}
   ]
  }
  MaterialExpressionMultiply_2 {
   class=MaterialExpressionMultiply
   inputs [
    A {from=MaterialExpressionReroute_0}
    B {from=MaterialExpressionScalarParameter_0}
   ]
  }
  MaterialExpressionMultiply_3 {
   class=MaterialExpressionMultiply
   inputs [
    A {from=MaterialExpressionScalarParameter_15}
    B {from=MaterialExpressionAppendVector_1}
   ]
  }
  MaterialExpressionReroute_0 {class=MaterialExpressionReroute, inputs [None {from=MaterialExpressionMaterialFunctionCall_2, output=RGB}]}
  MaterialExpressionScalarParameter_0 {class=MaterialExpressionScalarParameter, properties {defaultValue=0.3, group=09_Flipbook Mouth, parameterName=MouthEmissiveIntensity}}
  MaterialExpressionScalarParameter_1 {class=MaterialExpressionScalarParameter, properties {parameterName=flipbook_face_mouth_index}}
  MaterialExpressionScalarParameter_15 {class=MaterialExpressionScalarParameter, properties {defaultValue=2.5, group=09_Flipbook Mouth, parameterName=MouthScale}}
  MaterialExpressionScalarParameter_16 {class=MaterialExpressionScalarParameter, properties {defaultValue=-0.59, group=09_Flipbook Mouth, parameterName=MouthUVShiftX}}
  MaterialExpressionScalarParameter_17 {class=MaterialExpressionScalarParameter, properties {defaultValue=-1.5174, group=09_Flipbook Mouth, parameterName=MouthUVShiftY}}
  MaterialExpressionScalarParameter_2 {class=MaterialExpressionScalarParameter, properties {defaultValue=1, group=09_Flipbook Mouth, parameterName=MouthScale_Y}}
  MaterialExpressionSetMaterialAttributes_0 {
   class=MaterialExpressionSetMaterialAttributes
   inputs [
    MaterialAttributes {from=MaterialExpressionGetMaterialAttributes_0, output=MaterialAttributes}
    BaseColor {from=MaterialExpressionLinearInterpolate_15, output=""}
    EmissiveColor {from=MaterialExpressionStaticSwitchParameter_1, output=""}
   ]
   properties {attributeSetTypes [#1, #2]}
  }
  MaterialExpressionSphereMask_29 {
   class=MaterialExpressionSphereMask
   inputs [
    A {from=MaterialExpressionMaterialFunctionCall_36, output=G}
    B {from=MaterialExpressionConstant_11, output=""}
    Radius {from="", output=""}
    Hardness {from="", output=""}
   ]
   properties {attenuationRadius=0.5}
  }
  MaterialExpressionSphereMask_30 {
   class=MaterialExpressionSphereMask
   inputs [
    A {from=MaterialExpressionConstant_11, output=""}
    B {from=MaterialExpressionMaterialFunctionCall_36, output=R}
    Radius {from="", output=""}
    Hardness {from="", output=""}
   ]
   properties {attenuationRadius=0.5}
  }
  MaterialExpressionStaticSwitchParameter_0 {
   class=MaterialExpressionStaticSwitchParameter
   inputs [
    True {from=MaterialExpressionTextureCoordinate_1, output=""}
    False {from=MaterialExpressionTextureCoordinate_0, output=""}
   ]
   properties {group=09_Flipbook Mouth, parameterName=UseUV2forMouth}
  }
  MaterialExpressionStaticSwitchParameter_1 {
   class=MaterialExpressionStaticSwitchParameter
   inputs [
    True {from=MaterialExpressionLinearInterpolate_0, output=""}
    False {from=MaterialExpressionGetMaterialAttributes_0, output=EmissiveColor}
   ]
   properties {group=09_Flipbook Mouth, parameterName=useEmissiveforMouth}
  }
  MaterialExpressionSwitch_0 {
   class=MaterialExpressionSwitch
   inputs [
    SwitchValue {from=MaterialExpressionScalarParameter_1, output=""}
    Default {from="", output=""}
    None {from=MaterialExpressionScalarParameter_1, output=""}
    None {from=MaterialExpressionScalarParameter_1, output=""}
    None {from=MaterialExpressionScalarParameter_1, output=""}
    None {from=MaterialExpressionScalarParameter_1, output=""}
    None {from=MaterialExpressionConstant3Vector_0, output=R}
    None {from=MaterialExpressionConstant3Vector_0, output=R}
    None {from=MaterialExpressionConstant3Vector_0, output=R}
    None {from=MaterialExpressionConstant3Vector_1, output=R}
    None {from=MaterialExpressionConstant3Vector_1, output=R}
    None {from=MaterialExpressionConstant3Vector_1, output=R}
    None {from=MaterialExpressionConstant4Vector_0, output=R}
    None {from=MaterialExpressionConstant4Vector_0, output=R}
    None {from=MaterialExpressionConstant4Vector_0, output=R}
    None {from=MaterialExpressionConstant4Vector_0, output=R}
    None {from=MaterialExpressionConstant4Vector_0, output=R}
   ]
   properties {
    constDefault=1
    inputs [
     {
      inputName=None
      input {
       expression=&self:MaterialExpressionScalarParameter_1
       inputName=None
       mask=0
       maskA=0
       maskB=0
       maskG=0
       maskR=0
       outputIndex=0
      }
     }
     {
      inputName=None
      input {
       expression=&self:MaterialExpressionScalarParameter_1
       inputName=None
       mask=0
       maskA=0
       maskB=0
       maskG=0
       maskR=0
       outputIndex=0
      }
     }
     {
      inputName=None
      input {
       expression=&self:MaterialExpressionScalarParameter_1
       inputName=None
       mask=0
       maskA=0
       maskB=0
       maskG=0
       maskR=0
       outputIndex=0
      }
     }
     {
      inputName=None
      input {
       expression=&self:MaterialExpressionScalarParameter_1
       inputName=None
       mask=0
       maskA=0
       maskB=0
       maskG=0
       maskR=0
       outputIndex=0
      }
     }
     {
      inputName=None
      input {
       expression=&self:MaterialExpressionConstant3Vector_0
       inputName=None
       mask=1
       maskA=0
       maskB=0
       maskG=0
       maskR=1
       outputIndex=1
      }
     }
     {
      inputName=None
      input {
       expression=&self:MaterialExpressionConstant3Vector_0
       inputName=None
       mask=1
       maskA=0
       maskB=0
       maskG=1
       maskR=0
       outputIndex=2
      }
     }
     {
      inputName=None
      input {
       expression=&self:MaterialExpressionConstant3Vector_0
       inputName=None
       mask=1
       maskA=0
       maskB=1
       maskG=0
       maskR=0
       outputIndex=3
      }
     }
     {
      inputName=None
      input {
       expression=&self:MaterialExpressionConstant3Vector_1
       inputName=None
       mask=1
       maskA=0
       maskB=0
       maskG=0
       maskR=1
       outputIndex=1
      }
     }
     {
      inputName=None
      input {
       expression=&self:MaterialExpressionConstant3Vector_1
       inputName=None
       mask=1
       maskA=0
       maskB=0
       maskG=1
       maskR=0
       outputIndex=2
      }
     }
     {
      inputName=None
      input {
       expression=&self:MaterialExpressionConstant3Vector_1
       inputName=None
       mask=1
       maskA=0
       maskB=1
       maskG=0
       maskR=0
       outputIndex=3
      }
     }
     {
      inputName=None
      input {
       expression=&self:MaterialExpressionConstant4Vector_0
       inputName=None
       mask=1
       maskA=0
       maskB=0
       maskG=0
       maskR=1
       outputIndex=1
      }
     }
     {
      inputName=None
      input {
       expression=&self:MaterialExpressionConstant4Vector_0
       inputName=None
       mask=1
       maskA=0
       maskB=0
       maskG=1
       maskR=0
       outputIndex=2
      }
     }
     {
      inputName=None
      input {
       expression=&self:MaterialExpressionConstant4Vector_0
       inputName=None
       mask=1
       maskA=0
       maskB=0
       maskG=1
       maskR=0
       outputIndex=2
      }
     }
     {
      inputName=None
      input {
       expression=&self:MaterialExpressionConstant4Vector_0
       inputName=None
       mask=1
       maskA=0
       maskB=1
       maskG=0
       maskR=0
       outputIndex=3
      }
     }
     {
      inputName=None
      input {
       expression=&self:MaterialExpressionConstant4Vector_0
       inputName=None
       mask=1
       maskA=1
       maskB=0
       maskG=0
       maskR=0
       outputIndex=4
      }
     }
    ]
   }
  }
  MaterialExpressionTextureCoordinate_0 {class=MaterialExpressionTextureCoordinate}
  MaterialExpressionTextureCoordinate_1 {class=MaterialExpressionTextureCoordinate, properties {coordinateIndex=1}}
  MaterialExpressionTextureObjectParameter_0 {
   class=MaterialExpressionTextureObjectParameter
   inputs [
    UVs {from="", output=""}
    Apply View MipBias {from="", output=""}
   ]
   properties {
    group=09_Flipbook Mouth
    parameterName=MouthFlipbook
    texture=/Game/Characters/Player/Common/MaterialFunctions/VertexColorFaces/CharacterMouthFlipbookV2Template
   }
  }
 }}
