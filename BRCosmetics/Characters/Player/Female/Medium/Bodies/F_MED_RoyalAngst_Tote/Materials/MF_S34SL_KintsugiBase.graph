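# Super-level style material function for character cosmetics that layers a kintsugi gem-and-vein surface, accent and secondary print layers onto input material attributes by RGB mask.
# The clear-coat gem fakes inner glow with a refracted view vector sampling panning noise, plus Fresnel and parallax UVs.
# Metallic veins come from a crack texture triplanar-projected in pre-skinned local space.
# Static switches add WartyBrine and TealMink variants, gem skipping and mask preview.
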MaterialFunction /BRCosmetics/Characters/Player/Female/Medium/Bodies/F_MED_RoyalAngst_Tote/Materials/MF_S34SL_KintsugiBase
graph {nodes [
  n1 = NamedRerouteUsage()
  n2 = Constant3Vector() {constant=<rgba 0 0.023 0.0168 1>}
  n3 = FunctionInput(Preview=Constant3Vector() {constant=<rgba 0.1652 0.7552 0.5136 1>}) {inputName=GemInnerLightColor}
  n4 = MaterialFunctionCall(Hue Shift Percentage=ScalarParameter() {defaultValue=-0.1, parameterName=HueShiftPercentage}, Texture=@n3) {materialFunction=&HueShift}
  n5 = NamedRerouteUsage()
  n6 = TextureSampleParameter2D(UVs=NamedRerouteUsage()) {
   group=00_SuperLevels_T1/2
   parameterName=Gem Diffuse
   samplerSource=SSM_Wrap_WorldGroupSettings
   texture=/BRCosmetics/Characters/Player/Female/Medium/Bodies/F_MED_RoyalAngst_Tote/Textures/T_RoughGemSurface_D
  }
  n7 = CameraVectorWS()
  n8 = TextureSampleParameter2D(UVs=StaticSwitchParameter(True=NamedRerouteUsage(), False=NamedRerouteUsage()) {parameterName=IsWartyBrine}) {
   desc=Rock/Gem normal map
   group=00_SuperLevels_T1/2
   parameterName=Gem Normal
   samplerSource=SSM_Wrap_WorldGroupSettings
   samplerType=SAMPLERTYPE_Normal
   texture=/BRCosmetics/Characters/Player/Female/Medium/Bodies/F_MED_RoyalAngst_Tote/Textures/T_RoughGemSurface_N
  }
  n9 = LinearInterpolate(
   A=Constant3Vector() {constant=<rgba 0 0 1 0>}
   B=@n8.RGB
   Alpha=ScalarParameter() {defaultValue=0.3, group=00_SuperLevels_T1/2, parameterName=Refraction Normal Intensity}
  )
  n10 = MaterialFunctionCall!(Vector=MaterialFunctionCall!(
    Ray Direction=Multiply(A=@n7) {constB=-1}
    Surface Normal=@n9
    Refractive Index Origin
    Refractive Index Target=ScalarParameter() {defaultValue=1.15, group=00_SuperLevels_T1/2, parameterName=Refraction IOR}
   ) {materialFunction=/Engine/Functions/Engine_MaterialFunctions01/Vectors/Refract}.Vector) {materialFunction=/Engine/Functions/Engine_MaterialFunctions01/Coordinates/LongLatToUV}
  n11 = ScalarParameter() {defaultValue=4, group=00_SuperLevels_T1/2, parameterName=InnerNoise_Scale}
  n12 = Constant2Vector() {g=0.3, r=0.08}
  n13 = Multiply(A=Time(), B=ScalarParameter() {defaultValue=0.4, group=00_SuperLevels_T1/2, parameterName=InnerNoise_PanSpeed})
  n14 = TextureSample(UVs=Add(A=Multiply(A=@n10.Result, B=@n11), B=Multiply(A=@n12, B=@n13))) {
   desc=LowResBlurredNoiseNormalized
   samplerSource=SSM_Wrap_WorldGroupSettings
   samplerType=SAMPLERTYPE_Grayscale
   texture=/Game/Global/Textures/Default/Noise/LowResBlurredNoiseNormalized
  }
  n15 = ScalarParameter() {defaultValue=1, group=00_SuperLevels_T1/2, parameterName=InnerLight_Exponent, sliderMax=1}
  n16 = NamedRerouteUsage()
  n17 = Saturate(None=LinearInterpolate(A=Power(Base=@n14.RGB, Exp=@n15), Alpha=Power(Base=@n16) {constExponent=3}) {constB=-0.5})
  n18 = NamedRerouteUsage()
  n19 = Power(Base=NamedRerouteUsage())
  n20 = Saturate(None=Add(A=Multiply(A=@n19) {constB=50}) {constB=0.3})
  n21 = Multiply(
   A=Multiply(A=LinearInterpolate(A=@n3, B=@n4.Result, Alpha=@n5), B=@n6.RGB)
   B=StaticSwitchParameter(True=LinearInterpolate(A=@n17, Alpha=@n18) {constB=0.1}, False=Multiply(A=@n17, B=@n20)) {parameterName=IsWartyBrineHead}
  )
  n22 = Multiply(A=NamedRerouteUsage()) {constB=0.03}
  n23 = StaticSwitchParameter(True=LinearInterpolate(A=@n21, B=Add(A=@n21, B=@n22), Alpha=@n18) {constB=0.1}, False=@n21) {parameterName=IsWartyBrineHead}
  n24 = NamedRerouteUsage()
  n25 = NamedRerouteUsage()
  n26 = Multiply(A=@n23) {constB=4}
  n27 = MaterialFunctionCall!(Circumference Tiling, Position, Z Height, HQ Angle Calculation?) {materialFunction=/Game/Characters/Player/Common/MaterialFunctions/MF_CylindricalUVs_New}
  n28 = TextureObjectParameter() {
   parameterName="Gem Normal (WartyBrine)"
   samplerType=SAMPLERTYPE_Normal
   texture=&T_RoughGemSurface_N
  }
  n29 = Multiply(A=TextureCoordinate(), B=ScalarParameter() {defaultValue=1, group=00_SuperLevels_T1/2, parameterName=Gem UV Tiling})
  n30 = TextureSampleParameter2D(UVs=@n29) {
   desc=Rock/Gem normal map
   group=00_SuperLevels_T1/2
   parameterName=Gem Normal
   samplerSource=SSM_Wrap_WorldGroupSettings
   samplerType=SAMPLERTYPE_Normal
   texture=&T_RoughGemSurface_N
  }
  n31 = StaticSwitchParameter(True=TextureSample(UVs=@n27.UVs, Tex=@n28) {samplerSource=SSM_Wrap_WorldGroupSettings}.RGB, False=@n30.RGB) {parameterName=IsWartyBrine}
  n32 = ScalarParameter() {defaultValue=0.8, group=00_SuperLevels_T1/2, parameterName=Clear Coat Bottom Normal Intensity, sliderMax=2}
  n33 = LinearInterpolate(A=@n32, B=Constant() {r=0.2}, Alpha=NamedRerouteUsage())
  n34 = SetMaterialAttributes!(
   MaterialAttributes=NamedRerouteUsage()
   BaseColor=Multiply(A=FunctionInput(Preview=@n2) {inputName=GemSurfaceColor}, B=NamedRerouteUsage())
   Metallic=ScalarParameter() {group=00_SuperLevels_T1/2, parameterName=Gem_Metallic, sliderMax=1}
   Specular=ScalarParameter() {defaultValue=0.75, group=00_SuperLevels_T1/2, parameterName=Gem_Specular, sliderMax=1}
   Roughness=ScalarParameter() {defaultValue=0.22, group=00_SuperLevels_T1/2, parameterName=Gem_Roughness, sliderMax=1}
   EmissiveColor=LinearInterpolate(A=Add(A=@n23, B=@n24), B=LinearInterpolate(A=@n25, B=@n26) {constAlpha=0.95}, Alpha=NamedRerouteUsage())
   ShadingModel=ShadingModel() {shadingModel=MSM_ClearCoat}
   Normal=GetMaterialAttributes(None=NamedRerouteUsage()) {attributeGetTypes [#1]}.Normal
   ClearCoatBottomNormal=LinearInterpolate(
    A=Constant3Vector() {constant=<rgba 0 0 1 0>}
    B=NamedRerouteDeclaration(None=@n31) {name=Gem Inner Normal, nodeColor=<rgba 1 0 0.2578 1>}
    Alpha=StaticSwitchParameter(True=@n33, False=@n32) {parameterName=IsWartyBrineHead}
   )
   CustomData0=ScalarParameter() {defaultValue=0.8, group=00_SuperLevels_T1/2, parameterName=Gem_ClearCoatAmount, sliderMax=1}
   CustomData1=ScalarParameter() {defaultValue=0.1, group=00_SuperLevels_T1/2, parameterName=Gem_ClearCoatRoughness, sliderMax=1}
  ) {
   desc=Primary
   attributeSetTypes [
    #2
    #3
    #4
    #5
    #6
    #7
    #1
    #8
    #9
    #10
   ]
  }
  n35 = SetMaterialAttributes!(
   MaterialAttributes
   BaseColor=NamedRerouteUsage()
   Metallic=ScalarParameter() {defaultValue=1, group=00_SuperLevels_T1/2, parameterName=Veins_Metallic, sliderMax=1}
   Specular=ScalarParameter() {defaultValue=1, group=00_SuperLevels_T1/2, parameterName=Veins_Specular, sliderMax=1}
   Roughness=ScalarParameter() {defaultValue=0.1, group=00_SuperLevels_T1/2, parameterName=Veins_Roughness, sliderMax=1}
   Normal=MaterialFunctionCall(Normal=NamedRerouteUsage(), Flatness=Constant() {r=0.1}) {materialFunction=&FlattenNormal}.Result
   ShadingModel=ShadingModel()
  ) {attributeSetTypes [
    #2
    #3
    #4
    #5
    #1
    #7
   ]}
  n36 = NamedRerouteUsage()
  n37 = NamedRerouteUsage()
  n38 = BlendMaterialAttributes(
   A=@n34
   B=@n35
   Alpha=StaticSwitchParameter(True=LinearInterpolate(A=@n36, Alpha=@n37) {constB=0}, False=@n36) {parameterName=IsWartyBrineHead}
  )
  n39 = ComponentMask(None=NamedRerouteUsage()) {r=true}
  n40 = GetMaterialAttributes(None=StaticSwitchParameter(True=@n1, False=BlendMaterialAttributes(A=@n1, B=@n38, Alpha=@n39) {vertexAttributeBlendType=UseA}) {parameterName="SkipGem(Superlevel)"}) {attributeGetTypes [#1]}
  n41 = GetMaterialAttributes(None=NamedRerouteUsage()) {attributeGetTypes [#4, #5]}
  n42 = Add(A=@n41.Roughness, B=ScalarParameter() {group=00_SuperLevels_T1/2, parameterName=AccentRoughnessBias, sliderMax=1, sliderMin=-1})
  n43 = SetMaterialAttributes!(
   MaterialAttributes=@n41.MaterialAttributes
   BaseColor=FunctionInput(Preview=Constant3Vector() {constant=<rgba 0.8906 0.727 0.3433 1>}) {inputName=AccentColor}
   Metallic=ScalarParameter() {defaultValue=1, group=00_SuperLevels_T1/2, parameterName=AccentMetallic, sliderMax=1}
   Specular=Add(A=@n41.Specular, B=ScalarParameter() {group=00_SuperLevels_T1/2, parameterName=AccentSpecularBias, sliderMax=1, sliderMin=-1})
   Roughness=@n42
   EmissiveColor=Constant()
   ShadingModel=ShadingModel()
   Normal=@n40.Normal
  ) {attributeSetTypes [
    #2
    #3
    #4
    #5
    #6
    #7
    #1
   ]}
  n44 = NamedRerouteUsage()
  n45 = GetMaterialAttributes(None=NamedRerouteUsage()) {attributeGetTypes [#4, #1]}
  n46 = FunctionInput(Preview=Constant3Vector() {constant=<rgba 0.023 0.023 0.023 1>}) {inputName=SecondaryColor}
  n47 = Constant3Vector() {constant=<rgba 0 0 0 1>}
  n48 = NamedRerouteUsage()
  n49 = LinearInterpolate(Alpha=Power(Base=Saturate(None=@n48)) {constExponent=1}) {constA=0.8, constB=-1}
  n50 = NamedRerouteUsage()
  n51 = Multiply(
   A=MaterialFunctionCall(Normal Vector=VertexNormalWS(), Power=ScalarParameter() {defaultValue=0.5, parameterName=AdditionalPrintFresnel}) {materialFunction=&Fresnel_Function}.Result
   B=@n50
  ) {constA=1.5}
  n52 = SetMaterialAttributes!(
   MaterialAttributes=@n45.MaterialAttributes
   BaseColor=LinearInterpolate(A=@n46, B=FunctionInput(Preview=@n47) {inputName=PrintBaseColor}, Alpha=NamedRerouteUsage()) {constB=0}
   Metallic=ScalarParameter() {defaultValue=1, group=00_SuperLevels_T1/2, parameterName=SecondaryMetallic, sliderMax=1}
   Specular=ScalarParameter() {defaultValue=0.5, group=00_SuperLevels_T1/2, parameterName=SecondarySpecular, sliderMax=1}
   Roughness=Saturate(None=@n49)
   EmissiveColor=Multiply(A=@n46, B=Multiply(A=@n51, B=@n50) {constA=4})
   Normal=@n45.MaterialAttributes
  ) {attributeSetTypes [
    #2
    #3
    #4
    #5
    #6
    #1
   ]}
  n53 = BlendMaterialAttributes(
   A=BlendMaterialAttributes(A=@n40.MaterialAttributes, B=@n43, Alpha=ComponentMask(None=@n44) {g=true}) {vertexAttributeBlendType=UseA}
   B=@n52
   Alpha=ComponentMask(None=NamedRerouteUsage()) {b=true}
  ) {vertexAttributeBlendType=UseA}
  n54 = GetMaterialAttributes(None=@n53) {attributeGetTypes [#6]}
  n55 = NamedRerouteUsage()
  n56 = VectorParameter() {parameterName=WartyBrineGlassesColor, defaultValue=<rgba 0 1 0.295 1>}
  Add(A=@n54.EmissiveColor, B=Multiply(A=@n55, B=@n56.RGB))
  Constant()
  Constant()
  Constant()
  n57 = NamedRerouteUsage()
  n58 = NamedRerouteUsage()
  n59 = SetMaterialAttributes!(MaterialAttributes=@n53, EmissiveColor=LinearInterpolate(A=@n54.EmissiveColor, B=@n56.RGB, Alpha=@n55)) {attributeSetTypes [#6]}
  FunctionOutput(None=StaticSwitchParameter(
    True=SetMaterialAttributes!(MaterialAttributes=@n57, BaseColor=@n58, EmissiveColor=@n58) {attributeSetTypes [#2, #6]}
    False=StaticSwitchParameter(True=@n59, False=@n53) {parameterName=IsWartyBrineGlasses}
   ) {group=00_SuperLevels, parameterName=Preview Mask, sortPriority=40})
  MaterialFunctionCall!(BoundsScale) {materialFunction=/Game/Athena/SafeZone/S11/MF_CylindricalUVs}
  Multiply(A=@n49)
  n60 = NamedRerouteUsage()
  n61 = Constant3Vector() {constant=<rgba 0 0 1 0>}
  n62 = LinearInterpolate(
   A=MaterialFunctionCall(Normal=@n60, Flatness=Constant() {r=0.6}) {materialFunction=&FlattenNormal}.Result
   B=@n61
   Alpha=NamedRerouteUsage()
  )
  n63 = StaticSwitchParameter(True=LinearInterpolate(A=@n62, B=@n61, Alpha=NamedRerouteUsage()), False=@n62) {parameterName=IsWartyBrineHead}
  Multiply(A=@n63)
  n64 = TextureSampleParameter2D() {group=00_SuperLevels, parameterName=SuperLevel Mask, texture=&T_BlacksRGB}
  n65 = Constant() {r=1}
  n66 = ScalarParameter() {defaultValue=1, group=09_Flipbook Mouth, parameterName=MouthScale_Y}
  n67 = Multiply(
   A=StaticSwitchParameter(True=TextureCoordinate() {coordinateIndex=1}, False=TextureCoordinate()) {group=09_Flipbook Mouth, parameterName=UseUV2forMouth}
   B=Multiply(A=ScalarParameter() {defaultValue=2.5, group=09_Flipbook Mouth, parameterName=MouthScale}, B=AppendVector(A=@n65, B=@n66))
  )
  n68 = AppendVector(
   A=ScalarParameter() {defaultValue=-0.59, group=09_Flipbook Mouth, parameterName=MouthUVShiftX}
   B=ScalarParameter() {defaultValue=-1.5174, group=09_Flipbook Mouth, parameterName=MouthUVShiftY}
  )
  n69 = Subtract(
   A=@n64.RGB
   B=MaterialFunctionCall!(Material, MouthUVs=Add(A=@n67, B=@n68)) {materialFunction=/BRCosmetics/Characters/Player/Male/Medium/Bodies/M_MED_WartyBrine_Sus/Materials/MF_FV2_FlipbookMouths_WartyBrine}.Alpha
  )
  NamedRerouteDeclaration(None=StaticSwitchParameter(True=Saturate(None=@n69), False=@n64.RGB) {parameterName=IsWartyBrineHead}) {name=SL Mask, nodeColor=<rgba 0.7578 0 1 1>}
  n70 = LocalPosition() {includedOffsets=ExcludeOffsets, localOrigin=InstancePreSkinning}
  n71 = VertexInterpolator(VS=@n70.XYZ)
  n72 = MaterialFunctionCall(Float3=Multiply(A=ShaderStageSwitch(PixelShader=@n71.PS, VertexShader=@n70.XYZ), B=ScalarParameter() {defaultValue=0.01, parameterName=Tiling})) {materialFunction=&BreakOutFloat3Components}
  n73 = TextureObject() {samplerType=SAMPLERTYPE_Masks, texture=/BRCosmetics/Characters/Player/Female/Medium/Bodies/F_MED_RoyalAngst_Tote/Textures/T_GemCracks}
  n74 = FunctionInput(Preview=@n73) {inputName=VeinTexture, inputType=FunctionInput_Texture2D}
  n75 = TextureSample(
   UVs=MaterialFunctionCall(X=@n72.R, Y=@n72.R) {materialFunction=&MakeFloat2}.Result
   Tex=@n74
  ) {samplerType=SAMPLERTYPE_Masks, texture=&T_GemCracks}
  n76 = PreSkinnedNormal()
  n77 = ShaderStageSwitch(PixelShader=VertexInterpolator(VS=@n76).PS, VertexShader=@n76)
  n78 = MaterialFunctionCall(Float3=Power(Base=Abs(None=@n77), Exp=ScalarParameter() {defaultValue=3, parameterName=Power}) {constExponent=10}) {materialFunction=&BreakOutFloat3Components}
  n79 = TextureSample(
   UVs=MaterialFunctionCall(X=@n72.R, Y=@n72.R) {materialFunction=&MakeFloat2}.Result
   Tex=@n74
  ) {samplerType=SAMPLERTYPE_Masks, texture=&T_GemCracks}
  n80 = TextureSample(
   UVs=MaterialFunctionCall(X=@n72.G, Y=@n72.G) {materialFunction=&MakeFloat2}.Result
   Tex=@n74
  ) {samplerType=SAMPLERTYPE_Masks, texture=&T_GemCracks}
  n81 = Add(A=Add(A=Multiply(A=@n75.G, B=@n78.B), B=Multiply(A=@n79.G, B=@n78.G)), B=Multiply(A=@n80.G, B=@n78.R))
  NamedRerouteDeclaration(None=Power(Base=OneMinus(None=@n81)) {constExponent=4}) {name=TriplanarCells, nodeColor=<rgba 0 0.1797 1 1>}
  n82 = Multiply(A=@n75.R, B=@n78.B)
  n83 = Multiply(A=@n79.R, B=@n78.G)
  NamedRerouteDeclaration(None=Add(A=Add(A=@n82, B=@n83), B=Multiply(A=@n80.R, B=@n78.R))) {name=TriplanarGradient, nodeColor=<rgba 0.5 1 0 1>}
  NamedRerouteDeclaration(None=@n14.RGB) {name=LowFrequencyNoise, nodeColor=<rgba 1 0.1172 0 1>}
  n84 = TextureSampleParameter2D() {
   parameterName=AdditionalPrint
   samplerType=SAMPLERTYPE_Masks
   texture=/BRCosmetics/Characters/Player/Female/Medium/Bodies/F_MED_RoyalAngst_Tote/Skins/Print/Textures/T_RoyalAngst_TotePrint_Body_SL_FX2
  }
  n85 = StaticSwitchParameter(True=@n84.G, False=Constant()) {parameterName=HasAdditionalPrints}
  n86 = GetMaterialAttributes(None=NamedRerouteUsage()) {attributeGetTypes [#6, #2]}
  n87 = Constant() {r=1}
  n88 = Constant() {r=1}
  n89 = Power(
   Base=Max(A=Desaturation(None=@n86.BaseColor, Fraction=@n87), B=Desaturation(None=@n86.EmissiveColor, Fraction=@n88))
   Exp=ScalarParameter() {defaultValue=0.6, parameterName=AdditionalPrintPower}
  ) {constExponent=0.6}
  NamedRerouteDeclaration(None=Multiply(A=@n85, B=Add(A=@n85, B=@n89))) {name=AdditionalPrints, nodeColor=<rgba 0 0.5547 1 1>}
  NamedRerouteDeclaration(None=@n9) {name=LowFrequencyNormal, nodeColor=<rgba 0.1953 1 0 1>}
  NamedRerouteDeclaration(None=@n3) {name=InputGemInnerLightColor, nodeColor=<rgba 0.8047 0 1 1>}
  NamedRerouteDeclaration(None=MaterialFunctionCall(Power=ScalarParameter() {defaultValue=0.75, group=00_SuperLevels_T1/2, parameterName=GemFresnelPower2}) {materialFunction=&Fresnel_Function}.Result) {name=Fresnel_Simple, nodeColor=<rgba 0 1 0.9297 1>}
  n90 = LocalPosition() {localOrigin=InstancePreSkinning}
  n91 = VertexInterpolator(VS=@n90.XYZ)
  NamedRerouteDeclaration(None=SphereMask(A=ShaderStageSwitch(PixelShader=@n91.PS, VertexShader=@n90.XYZ), B=Constant3Vector() {constant=<rgba 0 10 155 0>}) {attenuationRadius=20, hardnessPercent=60}) {name=PickleFaceMask, nodeColor=<rgba 0 0.5547 1 1>}
  NamedRerouteDeclaration(None=@n27.UVs) {name=CylindricalUV, nodeColor=<rgba 0.5938 0 1 1>}
  n92 = Add(A=Multiply(A=@n75.B, B=@n78.B), B=Multiply(A=@n79.B, B=@n78.G))
  n93 = Multiply(A=@n80.B, B=@n78.R)
  n94 = OneMinus(None=Saturate(None=Add(A=@n92, B=@n93)))
  n95 = Power(Base=@n94) {constExponent=4}
  n96 = StaticSwitchParameter(True=LinearInterpolate(A=@n95, Alpha=NamedRerouteUsage()) {constB=0}, False=@n95) {parameterName=IsTealMinkHelmet}
  n97 = FunctionInput(Preview=Constant3Vector() {constant=<rgba 1 0.6526 0.3229 1>}) {inputName=VeinsColorLow}
  n98 = FunctionInput(Preview=Constant3Vector() {constant=<rgba 1 0.7346 0.3758 1>}) {inputName=VeinsColor}
  n99 = NamedRerouteUsage()
  NamedRerouteDeclaration(None=Multiply(A=NamedRerouteDeclaration(None=@n96) {name=VeinsMask, nodeColor=<rgba 0.2422 0 1 1>}, B=LinearInterpolate(A=@n97, B=@n98, Alpha=@n99))) {name=VeinsDiffuse, nodeColor=<rgba 0.0781 1 0 1>}
  NamedRerouteDeclaration(None=@n64.R) {name=TealMinkHelmetMask, nodeColor=<rgba 0.8516 0 1 1>}
  NamedRerouteDeclaration(None=@n64.G) {name=WartyBrineGlassesMask, nodeColor=<rgba 1 0.75 0 1>}
  NamedRerouteDeclaration(None=@n6.B) {name=Gem Diffuse R, nodeColor=<rgba 0.8047 0 1 1>}
  NamedRerouteDeclaration(None=@n94) {name=VeinsRaw, nodeColor=<rgba 1 0 0.3281 1>}
  NamedRerouteDeclaration(None=MaterialFunctionCall!(
    UV Coord=@n29
    Height=ScalarParameter() {defaultValue=1, group=00_SuperLevels_T1/2, parameterName=Parallax_Height, sliderMax=1}
    ReferencePlane=ScalarParameter() {defaultValue=1, group=00_SuperLevels_T1/2, parameterName=Parallax_RefPlane, sliderMax=1}
    HeightRatio=ScalarParameter() {defaultValue=0.05, group=00_SuperLevels_T1/2, parameterName=Parallax_HeightRatio}
    Normal=NamedRerouteUsage()
   ) {materialFunction=/Game/Global/MaterialFunctions/MF_BumpOffset_Accurate}.Result) {name=Gem Parallax UVs, nodeColor=<rgba 0.9219 1 0 1>}
  NamedRerouteDeclaration(None=FunctionInput() {inputType=FunctionInput_MaterialAttributes}) {name=InMA, nodeColor=<rgba 0 1 0.2266 1>}
  n100 = TextureCoordinate()
  n101 = ScalarParameter() {defaultValue=3, group=00_SuperLevels_T1/2, parameterName=VeinsScale}
  NamedRerouteDeclaration(None=Multiply(A=Multiply(A=@n100, B=@n101), B=ScalarParameter() {defaultValue=1, group=00_SuperLevels_T1/2, parameterName=TexelDensityScale})) {name=VeinsUV, nodeColor=<rgba 0.5469 1 0 1>}
  NamedRerouteDeclaration(None=Saturate(None=Multiply(A=@n94) {constB=4})) {name=VeinsMaskSoft, nodeColor=<rgba 1 0.0938 0 1>}
  n102 = FunctionInput(Preview=Constant3Vector() {constant=<rgba 0.1259 1 0.388 1>}) {inputName=GemFresnelColor}
  n103 = PixelNormalWS()
  n104 = MaterialFunctionCall!(BaseNormal=@n103, AdditionalNormal=Transform(None=@n63)) {materialFunction=/Engine/Functions/Engine_MaterialFunctions02/Utility/BlendAngleCorrectedNormals_WS}
  n105 = Add(A=MaterialFunctionCall(
    Normal Vector=StaticSwitchParameter(True=@n103, False=@n104.Result) {parameterName=IsWartyBrine}
    Invert Fresnel=StaticBool() {value=true}
    Power=ScalarParameter() {defaultValue=3, group=00_SuperLevels_T1/2, parameterName=GemFresnelPower}
   ) {materialFunction=&Fresnel_Function}.Result) {constB=0.02}
  NamedRerouteDeclaration(None=Multiply(A=Multiply(A=@n102, B=@n105), B=ScalarParameter() {defaultValue=0.5, group=00_SuperLevels_T1/2, parameterName=FresnelAmount})) {name=GemFresnel, nodeColor=<rgba 0.9219 1 0 1>}
  NamedRerouteUsage()
  NamedRerouteUsage()
  Saturate(None=Multiply(A=@n94) {constB=20})
  Saturate(None=@n42)
  ScalarParameter() {defaultValue=0.3, group=00_SuperLevels_T1/2, parameterName=SecondaryRoughness, sliderMax=1}
  Transform(None=@n60)
  VectorParameter() {group=00_SuperLevels_T1/2, parameterName=AccentColor, defaultValue=<rgba 0.8906 0.727 0.3433 1>}
  VectorParameter() {parameterName=GemFresnelColor, defaultValue=<rgba 0.1259 1 0.388 1>}
  VertexNormalWS()
 ]}
