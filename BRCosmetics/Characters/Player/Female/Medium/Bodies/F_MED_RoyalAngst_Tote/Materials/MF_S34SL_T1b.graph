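# Material function for the T1b Super Levels style of the F_MED_RoyalAngst_Tote character body, applying a kintsugi-like vein and gem look.
# It acts as a preset wrapper that passes incoming material attributes through MF_S34SL_KintsugiBase and outputs the result.
# Exposed color parameters for veins, gem surface, inner light, fresnel, accent and print feed that base, along with a T_GemCracks mask texture as the vein pattern.
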MaterialFunction /BRCosmetics/Characters/Player/Female/Medium/Bodies/F_MED_RoyalAngst_Tote/Materials/MF_S34SL_T1b
graph {nodes [
  Constant() {r=-1}
  n1 = TextureObjectParameter() {
   parameterName=VeinTexture_T1b
   samplerType=SAMPLERTYPE_Masks
   texture=/BRCosmetics/Characters/Player/Female/Medium/Bodies/F_MED_RoyalAngst_Tote/Textures/T_GemCracks
  }
  FunctionOutput(None=MaterialFunctionCall!(
    VeinsColorLow=VectorParameter() {group=00_SuperLevels_T1/2, parameterName=VeinsColorLow, defaultValue=<rgba 1.6 1 0.4375 1>}.RGB
    AccentColor=VectorParameter() {group=00_SuperLevels_T1/2, parameterName=AccentColor, defaultValue=<rgba 0.6354 0.3703 0.1693 1>}.RGB
    GemSurfaceColor=VectorParameter() {group=00_SuperLevels_T1/2, parameterName=GemSurfaceColor, defaultValue=<rgba 0 0.026 0.025 1>}.RGB
    GemInnerLightColor=VectorParameter() {group=00_SuperLevels_T1/2, parameterName=GemInnerLightColor, defaultValue=<rgba 0.136 2 1.5 1>}.RGB
    VeinsColor=VectorParameter() {group=00_SuperLevels_T1/2, parameterName=VeinsColor, defaultValue=<rgba 0.1875 0.1153 0.0176 1>}.RGB
    In=FunctionInput() {inputType=FunctionInput_MaterialAttributes}
    GemFresnelColor=VectorParameter() {parameterName=GemFresnelColor, defaultValue=<rgba 0.0126 0.3021 0.2141 1>}.RGB
    SecondaryColor=VectorParameter() {parameterName=SecondaryColor_T1, defaultValue=<rgba 0.0469 0.0469 0.0469 1>}.RGB
    VeinTexture=@n1
    PrintBaseColor=VectorParameter() {parameterName=PrintBaseColorT1, defaultValue=<rgba 0 0 0 1>}.RGB
   ) {materialFunction=&/MF_S34SL_KintsugiBase}.Result)
 ]}
