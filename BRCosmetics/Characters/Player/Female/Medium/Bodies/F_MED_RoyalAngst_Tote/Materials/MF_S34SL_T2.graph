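# Material function that applies the T2 color tier of a gem-and-veins "Kintsugi" look for the RoyalAngst Tote female player body.
# It passes incoming material attributes through MF_S34SL_KintsugiBase, feeding it tier-specific color parameters for veins, accents, gem surface, inner light, fresnel and secondary color, plus a crack mask texture.
# The print base color first runs through the shared MF_TODColorCurve function before reaching the base.
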MaterialFunction /BRCosmetics/Characters/Player/Female/Medium/Bodies/F_MED_RoyalAngst_Tote/Materials/MF_S34SL_T2
graph {nodes [
  n1 = TextureObjectParameter() {
   parameterName=VeinTexture
   samplerType=SAMPLERTYPE_Masks
   texture=/BRCosmetics/Characters/Player/Female/Medium/Bodies/F_MED_RoyalAngst_Tote/Textures/T_GemCracks_B
  }
  n2 = VectorParameter() {parameterName=PrintBaseColor_T2, defaultValue=<rgba 0.1927 0.1173 0.0621 1>}
  FunctionOutput(None=MaterialFunctionCall!(
    VeinsColorLow=VectorParameter() {group=00_SuperLevels_T1/2, parameterName=VeinsColorLow_T2, defaultValue=<rgba 2 0.9006 0.8438 1>}.RGB
    AccentColor=VectorParameter() {group=00_SuperLevels_T1/2, parameterName=AccentColor_T2, defaultValue=<rgba 1 0.5846 0.4635 1>}.RGB
    GemSurfaceColor=VectorParameter() {group=00_SuperLevels_T1/2, parameterName=GemSurfaceColor_T2, defaultValue=<rgba 0.1406 0.0066 0 1>}.RGB
    GemInnerLightColor=VectorParameter() {group=00_SuperLevels_T1/2, parameterName=GemInnerLightColor_T2, defaultValue=<rgba 4 0.05 0.05 1>}.RGB
    VeinsColor=VectorParameter() {group=00_SuperLevels_T1/2, parameterName=VeinsColor_T2, defaultValue=<rgba 0.2656 0.1134 0.1032 1>}.RGB
    In=FunctionInput() {inputType=FunctionInput_MaterialAttributes}
    GemFresnelColor=VectorParameter() {parameterName=GemFresnelColor_T2, defaultValue=<rgba 0.2188 0.0116 0.0141 1>}.RGB
    SecondaryColor=VectorParameter() {parameterName=SecondaryColor_T2, defaultValue=<rgba 0.4479 0.2905 0.1213 1>}.RGB
    VeinTexture=@n1
    PrintBaseColor=MaterialFunctionCall!(Color=@n2.RGB) {materialFunction=/Game/Packages/Fortress_SharedMaterials/Anime_Master_Material/MF_TODColorCurve}.Result
   ) {materialFunction=&/MF_S34SL_KintsugiBase}.Result)
 ]}
