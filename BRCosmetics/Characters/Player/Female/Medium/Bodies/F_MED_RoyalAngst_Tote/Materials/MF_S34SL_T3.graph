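# Super-level T3 style material function for the RoyalAngst Tote cosmetic, layering primary, secondary and opal looks over incoming material attributes using SuperLevel Mask channels.
# Primary adds drifting noise color, fresnel and screen-space glitter anchored to the object; secondary defaults to flat-colored metal.
# Clear-coat opal fakes depth using bump offset and refracted, panning noise, and is skippable.
# WartyBrine switches add glasses glow, mouth-flipbook mask cutout and head-specific emissive.
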MaterialFunction /BRCosmetics/Characters/Player/Female/Medium/Bodies/F_MED_RoyalAngst_Tote/Materials/MF_S34SL_T3
graph {nodes [
  n1 = PreSkinnedNormal()
  n2 = Abs(None=ShaderStageSwitch(PixelShader=VertexInterpolator(VS=@n1).PS, VertexShader=@n1))
  n3 = Subtract(A=ComponentMask(None=@n2) {r=true}, B=ComponentMask(None=@n2) {g=true})
  Ceil(None=@n3)
  Constant3Vector() {constant=<rgba 0 0 1 0>}
  Constant()
  n4 = FunctionInput() {inputType=FunctionInput_MaterialAttributes}
  n5 = NamedRerouteDeclaration(None=@n4) {name=Mat_In, nodeColor=<rgba 0 1 0.9297 1>}
  n6 = BlendMaterialAttributes(A=@n4, B=NamedRerouteUsage(), Alpha=ComponentMask(None=NamedRerouteUsage()) {b=true}) {vertexAttributeBlendType=UseA}
  n7 = NamedRerouteUsage()
  n8 = BlendMaterialAttributes(
   A=BlendMaterialAttributes(
    A=StaticSwitchParameter(True=@n5, False=@n6) {group=00_SuperLevels_T3, parameterName=SkipOpal}
    B=NamedRerouteUsage()
    Alpha=ComponentMask(None=@n7) {g=true}
   ) {vertexAttributeBlendType=UseA}
   B=NamedRerouteUsage()
   Alpha=ComponentMask(None=NamedRerouteUsage()) {r=true}
  ) {vertexAttributeBlendType=UseA}
  n9 = Add(
   A=GetMaterialAttributes(None=@n8) {attributeGetTypes [#1]}.EmissiveColor
   B=Multiply(A=NamedRerouteUsage(), B=VectorParameter() {parameterName=WartyBrineGlassesColor, defaultValue=<rgba 0 1 0.295 1>}.RGB)
  )
  FunctionOutput(None=StaticSwitchParameter(True=SetMaterialAttributes!(MaterialAttributes=@n8, EmissiveColor=@n9) {attributeSetTypes [#1]}, False=@n8) {parameterName=IsWartyBrineGlasses})
  LinearInterpolate(A=Constant3Vector() {constant=<rgba 0 0 1 0>}, Alpha=ScalarParameter() {defaultValue=0.216, parameterName=Refraction Normal Intensity})
  MaterialFunctionCall!(Basecolor Texture, Worldpos, TextureSize) {materialFunction=/Game/Athena/Environments/MaterialFunctions/SimpleLowEndTriplanar}
  MaterialFunctionCall() {materialFunction=&ObjectPivotPoint}
  n10 = Constant3Vector() {constant=<rgba 0 0 120 0>}
  n11 = TransformPosition(None=@n10) {transformType=TRANSFORMPOSSOURCE_World}
  n12 = Subtract(
   A=ConstantBiasScale(None=ScreenPosition().ViewportUV) {bias=-0.5, scale=2}
   B=Multiply(
    A=Constant2Vector() {desc="The 'V' of screen space coordinate is opposite of the \"Y\" of clip space coordinate", g=-1, r=1}
    B=MaterialFunctionCall(World Position=@n11) {materialFunction=&TransformToClipSpace}.Clip Space XY
   )
  )
  n13 = MaterialFunctionCall(Float2=ViewSize()) {materialFunction=&BreakOutFloat2Components}
  n14 = AppendVector(A=Constant() {r=1}, B=Divide(A=@n13.G, B=@n13.G))
  n15 = ScalarParameter() {defaultValue=1, parameterName=ScaleU}
  n16 = ScalarParameter() {defaultValue=1, parameterName=Scalev}
  n17 = Multiply(
   A=Multiply(A=Multiply(A=@n12, B=@n14), B=AppendVector(A=@n15, B=@n16))
   B=ComponentMask(None=TransformPosition(None=@n11) {transformSourceType=TRANSFORMPOSSOURCE_World, transformType=TRANSFORMPOSSOURCE_View}) {b=true}
  )
  n18 = TextureSample(UVs=Divide(A=@n17, B=ScalarParameter() {defaultValue=24, parameterName=GlitterScale})) {desc=Glitter, samplerType=SAMPLERTYPE_Masks, texture=/Game/Global/Materials/MaterialGym/Materials/Glitter/Textures/T_Glitter_TA}
  Multiply(A=@n18.G)
  n19 = CameraVectorWS()
  n20 = BumpOffset(Height=Constant() {r=0.6}, HeightRatioInput=ScalarParameter() {defaultValue=0.05, parameterName=InnerHeightRatio}) {heightRatio=0.15, referencePlane=1}
  n21 = TextureSampleParameter2D(UVs=@n20) {
   desc=Rock/Gem normal map
   parameterName=Gem Normal
   samplerSource=SSM_Wrap_WorldGroupSettings
   samplerType=SAMPLERTYPE_Normal
   texture=/BRCosmetics/Characters/Player/Female/Medium/Bodies/F_MED_RoyalAngst_Tote/Textures/T_Opal_B_N
  }
  n22 = MaterialFunctionCall!(Vector=MaterialFunctionCall!(
    Ray Direction=Multiply(A=@n19) {constB=-1}
    Surface Normal=@n21.RGB
    Refractive Index Origin
    Refractive Index Target=ScalarParameter() {defaultValue=1.2, parameterName=Refraction IOR}
   ) {materialFunction=/Engine/Functions/Engine_MaterialFunctions01/Vectors/Refract}.Vector) {materialFunction=/Engine/Functions/Engine_MaterialFunctions01/Coordinates/LongLatToUV}
  n23 = Multiply(A=Constant2Vector() {g=0.01, r=0.003}, B=Time())
  n24 = ScalarParameter() {defaultValue=2, parameterName=InnerNoiseTiling}
  n25 = Constant2Vector() {g=2, r=1}
  n26 = TextureCoordinate()
  n27 = TextureSample!(
   UVs=Multiply(A=Add(A=@n22.Result, B=@n23) {desc=Add some panning if you like}, B=Multiply(A=@n24, B=@n25))
   Tex
   "DDX(UVs)"=DDX(Value=@n26)
   "DDY(UVs)"=DDY(Value=@n26)
   Apply View MipBias
  ) {
   desc=LowResBlurredNoiseNormalized
   mipValueMode=TMVM_Derivative
   samplerSource=SSM_Wrap_WorldGroupSettings
   samplerType=SAMPLERTYPE_Grayscale
   texture=/Game/Global/Textures/Default/Noise/LowResBlurredNoiseNormalized
  }
  n28 = TextureSample!(
   UVs=StaticSwitchParameter(
    True=MaterialFunctionCall!(Circumference Tiling, Position, Z Height, HQ Angle Calculation?) {materialFunction=/Game/Characters/Player/Common/MaterialFunctions/MF_CylindricalUVs_New}.UVs
    False=NamedRerouteUsage()
   ) {parameterName=UseCylindricalUVs}
   Tex
   Apply View MipBias
  ) {
   samplerSource=SSM_Wrap_WorldGroupSettings
   texture=/BRCosmetics/Characters/Player/Female/Medium/Bodies/F_MED_RoyalAngst_Tote/Textures/T_Opal_B_D
  }
  n29 = TextureSample(UVs=@n20) {
   samplerSource=SSM_Wrap_WorldGroupSettings
   texture=&T_Opal_B_D
  }
  n30 = ComponentMask(None=LocalPosition() {localOrigin=InstancePreSkinning}.XYZ) {r=true}
  n31 = ShaderStageSwitch(PixelShader=VertexInterpolator(VS=@n30).PS, VertexShader=@n30)
  n32 = Saturate(None=Divide(A=Abs(None=@n31)) {constB=10})
  n33 = Multiply(
   A=@n27.RGB
   B=StaticSwitchParameter(
    True=LinearInterpolate(A=@n28.RGB, B=@n29.RGB, Alpha=@n32)
    False=StaticSwitchParameter(True=@n28.RGB, False=@n29.RGB) {parameterName=UseTriplanarforOpal}
   ) {parameterName=IsWartyBrineGlasses}
  )
  n34 = MaterialFunctionCall(Invert Fresnel=StaticBool() {value=true}, Power=Constant() {r=6}) {materialFunction=&Fresnel_Function}
  n35 = Multiply(A=Multiply(A=@n33, B=ScalarParameter() {defaultValue=2, parameterName=Inner Brightness}), B=Saturate(None=Add(A=@n34.Result) {constB=0.1}))
  n36 = Multiply(A=@n35) {constB=5}
  n37 = Constant3Vector() {constant=<rgba 0.115 0 1.5 1>}
  n38 = SetMaterialAttributes!(
   MaterialAttributes
   BaseColor=LinearInterpolate(A=@n33, B=Multiply(B=@n33) {constA=3}, Alpha=NamedRerouteUsage()) {constA=1}
   EmissiveColor=LinearInterpolate(A=@n35, B=Add(A=@n36, B=@n37), Alpha=NamedRerouteUsage())
   ClearCoatBottomNormal=@n21.RGB
   ShadingModel=ShadingModel() {shadingModel=MSM_ClearCoat}
   Normal=NamedRerouteUsage()
  ) {attributeSetTypes [
    #2
    #1
    #3
    #4
    #5
   ]}
  NamedRerouteDeclaration(None=@n38) {name=Gem, nodeColor=<rgba 1 0 0.75 1>}
  n39 = TextureSampleParameter2D() {
   group=00_SuperLevels
   parameterName=SuperLevel Mask
   samplerSource=SSM_Wrap_WorldGroupSettings
   texture=&T_BlacksRGB
  }
  n40 = Constant() {r=1}
  n41 = ScalarParameter() {defaultValue=1, group=09_Flipbook Mouth, parameterName=MouthScale_Y}
  n42 = Multiply(
   A=StaticSwitchParameter(True=TextureCoordinate() {coordinateIndex=1}, False=TextureCoordinate()) {group=09_Flipbook Mouth, parameterName=UseUV2forMouth}
   B=Multiply(A=ScalarParameter() {defaultValue=2.5, group=09_Flipbook Mouth, parameterName=MouthScale}, B=AppendVector(A=@n40, B=@n41))
  )
  n43 = AppendVector(
   A=ScalarParameter() {defaultValue=-0.59, group=09_Flipbook Mouth, parameterName=MouthUVShiftX}
   B=ScalarParameter() {defaultValue=-1.5174, group=09_Flipbook Mouth, parameterName=MouthUVShiftY}
  )
  n44 = Subtract(
   A=@n39.RGB
   B=MaterialFunctionCall!(Material, MouthUVs=Add(A=@n42, B=@n43)) {materialFunction=/BRCosmetics/Characters/Player/Male/Medium/Bodies/M_MED_WartyBrine_Sus/Materials/MF_FV2_FlipbookMouths_WartyBrine}.Alpha
  )
  NamedRerouteDeclaration(None=StaticSwitchParameter(True=Saturate(None=@n44), False=@n39.RGB) {group=WartyBrine, parameterName=IsWartyBrineHead}) {name=SL Mask, nodeColor=<rgba 0.7578 0 1 1>}
  n45 = LocalPosition() {localOrigin=InstancePreSkinning}
  n46 = Constant() {r=64}
  n47 = Divide(A=ShaderStageSwitch(PixelShader=VertexInterpolator(VS=@n45.XYZ).PS, VertexShader=@n45.XYZ), B=Multiply(A=Abs(None=@n46)) {constB=-1})
  n48 = LinearInterpolate(
   A=ComponentMask(None=@n47) {b=true, r=true}
   B=ComponentMask(None=@n47) {b=true, g=true}
   Alpha=Saturate(None=Power(Base=@n3) {constExponent=0.1})
  )
  n49 = TextureObject() {
   samplerType=SAMPLERTYPE_Normal
   texture=/BRCosmetics/Characters/Player/Female/Medium/Bodies/F_MED_RoyalAngst_Tote/Textures/T_RoughGemSurface_N
  }
  n50 = Multiply(A=TextureCoordinate(), B=ScalarParameter() {defaultValue=0.25, parameterName=MipDerivativeScale})
  n51 = DDX(Value=@n50)
  n52 = DDY(Value=@n50)
  n53 = Fresnel(Normal=NamedRerouteDeclaration(None=TextureSample!(UVs=@n48, Tex=@n49, "DDX(UVs)"=@n51, "DDY(UVs)"=@n52, Apply View MipBias) {desc=Y projection, mipValueMode=TMVM_Derivative, samplerSource=SSM_Wrap_WorldGroupSettings}.RGB) {name=TriplanrNormal, nodeColor=<rgba 0 1 0.8359 1>}) {exponent=1}
  NamedRerouteDeclaration(None=Multiply(B=OneMinus(None=@n53)) {constA=0.1}) {name=WartyBrineFakeFresnel, nodeColor=<rgba 0 0.3906 1 1>}
  NamedRerouteDeclaration(None=@n39.G) {name=WartyBrineGlassMask, nodeColor=<rgba 0.7109 1 0 1>}
  NamedRerouteDeclaration(None=@n48) {name=TriplanarUV, nodeColor=<rgba 0 1 0.4844 1>}
  n54 = TextureSampleParameter2D() {
   parameterName=AdditionalPrint
   samplerSource=SSM_Wrap_WorldGroupSettings
   samplerType=SAMPLERTYPE_Masks
   texture=/BRCosmetics/Characters/Player/Female/Medium/Bodies/F_MED_RoyalAngst_Tote/Skins/Print/Textures/T_RoyalAngst_TotePrint_Body_SL_FX2
  }
  n55 = StaticSwitchParameter(True=@n54.G, False=Constant()) {parameterName=HasAdditionalPrints}
  n56 = GetMaterialAttributes(None=NamedRerouteUsage()) {attributeGetTypes [#1, #2]}
  n57 = Max(A=Desaturation(None=@n56.BaseColor, Fraction=Constant() {r=1}), B=Desaturation(None=@n56.EmissiveColor, Fraction=Constant() {r=1}))
  NamedRerouteDeclaration(None=Multiply(A=@n55, B=Add(A=@n55, B=@n57))) {name=AdditionalPrints, nodeColor=<rgba 0 0.5547 1 1>}
  n58 = LocalPosition() {localOrigin=InstancePreSkinning}
  n59 = VertexInterpolator(VS=@n58.XYZ)
  NamedRerouteDeclaration(None=SphereMask(A=ShaderStageSwitch(PixelShader=@n59.PS, VertexShader=@n58.XYZ), B=Constant3Vector() {constant=<rgba 0 10 160 0>}) {attenuationRadius=20, hardnessPercent=60}) {name=PickleFaceMask, nodeColor=<rgba 0 0.5547 1 1>}
  NamedRerouteDeclaration(None=Multiply(A=TextureCoordinate(), B=ScalarParameter() {defaultValue=4, parameterName=Gem UV Tiling})) {name=Gem UVs, nodeColor=<rgba 0.9219 1 0 1>}
  n60 = GetMaterialAttributes(None=NamedRerouteUsage()) {attributeGetTypes [#6, #5]}
  n61 = SetMaterialAttributes!(
   MaterialAttributes=@n60.MaterialAttributes
   BaseColor=VectorParameter() {parameterName=SecondaryColor_T3, defaultValue=<rgba 0.4688 0.4688 0.4688 1>}.RGB
   Metallic=ScalarParameter() {defaultValue=1, group=00_SuperLevels_T3, parameterName=SecondaryMetallic_T3, sliderMax=1}
   Specular=ScalarParameter() {defaultValue=0.8, group=00_SuperLevels_T3, parameterName=SecondarySpecular_T3, sliderMax=1}
   Roughness=ScalarParameter() {defaultValue=0.15, group=00_SuperLevels_T3, parameterName=SecondaryRoughness_T3, sliderMax=1}
   EmissiveColor
   Normal=@n60.MaterialAttributes
  ) {attributeSetTypes [
    #2
    #7
    #6
    #8
    #1
    #5
   ]}
  NamedRerouteDeclaration(None=@n61) {name=Secondary, nodeColor=<rgba 1 0.6094 0 1>}
  n62 = GetMaterialAttributes(None=NamedRerouteUsage()) {attributeGetTypes [#6, #5]}
  n63 = TextureSampleParameter2D(UVs=TextureCoordinate() {uTiling=1.5, vTiling=1.5}) {
   parameterName=StoneNormal
   samplerSource=SSM_Wrap_WorldGroupSettings
   samplerType=SAMPLERTYPE_Normal
   texture=&T_RoughGemSurface_N
  }
  n64 = StaticSwitchParameter(True=Constant3Vector() {constant=<rgba 0 0 1 0>}, False=@n63.RGB) {group=WartyBrine, parameterName=IsWartyBrine}
  n65 = ScalarParameter() {defaultValue=0.6, parameterName=GemNormalFlatten}
  n66 = SetMaterialAttributes!(
   MaterialAttributes=@n62.MaterialAttributes
   BaseColor=Constant()
   Metallic=ScalarParameter() {group=00_SuperLevels_T3, parameterName=PrimaryMetallic_T3, sliderMax=1}
   Specular=ScalarParameter() {defaultValue=0.2, group=00_SuperLevels_T3, parameterName=PrimarySpecular_T3, sliderMax=1}
   Roughness=ScalarParameter() {defaultValue=0.6, group=00_SuperLevels_T3, parameterName=PrimaryRoughness_T3, sliderMax=1}
   EmissiveColor=NamedRerouteUsage()
   Normal=MaterialFunctionCall(
    BaseNormal=@n62.Normal
    AdditionalNormal=MaterialFunctionCall(Normal=@n64, Flatness=@n65) {materialFunction=&FlattenNormal}.Result
   ) {materialFunction=&BlendAngleCorrectedNormals}.Result
   SubsurfaceColor=Constant()
  ) {attributeSetTypes [
    #2
    #7
    #6
    #8
    #1
    #5
    #9
   ]}
  NamedRerouteDeclaration(None=@n66) {name=Primary, nodeColor=<rgba 1 0 0.9141 1>}
  n67 = VectorParameter() {parameterName=PrimaryColor2_T3, defaultValue=<rgba 0.0908 0.0359 0.1302 1>}
  n68 = ScalarParameter() {defaultValue=0.1, parameterName=PrimaryNoiseDistortionAmount}
  n69 = Divide(A=@n17, B=Constant() {r=100})
  n70 = Multiply(A=Time()) {constB=0.02}
  n71 = TextureSample!(
   UVs=MaterialFunctionCall(UVs=@n69, Rotation Center=Constant(), "Rotation Angle (0-1)"=Frac(None=@n70)) {materialFunction=&CustomRotator}.Rotated Values
   Tex
   Apply View MipBias
  ) {samplerSource=SSM_Wrap_WorldGroupSettings, texture=/Game/Effects/Fort_Effects/Textures/Tile/T_GaussianNoise_01_M}
  n72 = TextureSample(UVs=Add(A=Multiply(A=@n68, B=@n71.R), B=@n69)) {samplerSource=SSM_Wrap_WorldGroupSettings, texture=&T_GaussianNoise_01_M}
  n73 = ScalarParameter() {defaultValue=1, parameterName=PrimaryColorMixpPower}
  n74 = MaterialFunctionCall(Normal Vector=PixelNormalWS(), Power=ScalarParameter() {defaultValue=2, group=00_SuperLevels_T1/2, parameterName=PrimaryFresnelPower}) {materialFunction=&Fresnel_Function}
  n75 = Add(A=@n74.Result, B=NamedRerouteUsage())
  n76 = Add(
   A=LinearInterpolate(A=VectorParameter() {parameterName=PrimaryColor_T3, defaultValue=<rgba 0 0 0 1>}.RGB, B=@n67.RGB, Alpha=Power(Base=@n72.R, Exp=@n73))
   B=Multiply(
    A=VectorParameter() {parameterName=PrimaryFresnelColor_T3, defaultValue=<rgba 0.1925 0.1707 0.2292 1>}.RGB
    B=StaticSwitchParameter(True=@n75, False=@n74.Result) {group=WartyBrine, parameterName=IsWartyBrine}
   )
  )
  n77 = Multiply(B=@n76) {constA=4}
  n78 = NamedRerouteUsage()
  n79 = Constant() {r=0.02}
  n80 = SphereMask(
   A=Power(Base=@n18.R) {constExponent=1}
   B=MaterialFunctionCall(
    Input=MaterialFunctionCall(Speed=@n79) {materialFunction=&DebugTimeSine}.Result
    Input Low=Constant()
    Input High=Constant() {r=1}
    Target Low=Constant() {r=0.15}
    Target High=Constant() {r=0.6}
   ) {materialFunction=&RemapValueRange}.Result
  ) {attenuationRadius=0.015, hardnessPercent=0}
  n81 = SphereMask(A=PixelDepth().R, B=Constant(), Radius=ScalarParameter() {defaultValue=500, parameterName=GlitterDistanceFade}) {hardnessPercent=10}
  n82 = Multiply(
   A=Saturate(None=Multiply(A=@n18.G, B=@n80))
   B=LinearInterpolate(
    A=ScalarParameter() {defaultValue=1, parameterName=GlitterEmissiveMult_Near}
    B=ScalarParameter() {defaultValue=5, parameterName=GlitterEmissiveMult_Far}
    Alpha=OneMinus(None=@n81)
   )
  )
  n83 = NamedRerouteUsage()
  n84 = Add(
   A=NamedRerouteDeclaration(None=LinearInterpolate(A=@n76, B=@n77, Alpha=@n78)) {name=GemFresnel, nodeColor=<rgba 0.9219 1 0 1>}
   B=Multiply(A=LinearInterpolate(A=@n82, Alpha=@n83) {constB=0})
  )
  n85 = Add(A=@n84, B=Desaturation(None=@n67.RGB, Fraction=Constant() {r=0.5}))
  n86 = NamedRerouteUsage()
  NamedRerouteDeclaration(None=StaticSwitchParameter(True=LinearInterpolate(A=@n84, B=@n85, Alpha=@n86), False=@n84) {group=WartyBrine, parameterName=IsWartyBrineHead}) {name=PrimaryeEmissiveColor, nodeColor=<rgba 0.9219 0 1 1>}
  NamedRerouteDeclaration(None=GetMaterialAttributes(None=@n5) {attributeGetTypes [#5]}.Normal) {name=Normal_In, nodeColor=<rgba 1 0.6328 0 1>}
  NamedRerouteUsage()
  NamedRerouteUsage()
  Step(Y=LocalPosition().Z) {constX=110}
  Transform(None=@n10) {transformSourceType=TRANSFORMSOURCE_Local}
  Transform(None=NamedRerouteUsage())
  VertexNormalWS()
 ]}
