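# Parent character material for S34SL Super Level cosmetics, stored with the female RoyalAngst_Tote body, supporting multiple shading models through material attributes.
# Shared Character Parent 2023 PreFX output feeds switchable hair, fly-away hair and advanced clear coat paths, with a metal-masked roughness multiplier.
# A masked, switchable base-color desaturation prepares attributes for static-switched Super Level tier functions (T1b, T2, T3).
# Shared PostFX follows, with separate normal handling for clear coat.
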Material /BRCosmetics/Characters/Player/Female/Medium/Bodies/F_MED_RoyalAngst_Tote/Materials/M_S34SL_Parent
graph {
 properties {
  bUseMaterialAttributes=true
  bUsedWithMeshDeformer=true
  bUsedWithMorphTargets=true
  bUsedWithSkeletalMesh=true
  pixelDepthOffsetMode=PDOM_Legacy
  shadingModel=MSM_FromMaterialExpression
  usedShadingModels=Default Lit | Subsurface | Clear Coat | Subsurface Profile | Cloth
 }
 nodes [
  n1 = GetMaterialAttributes(None=NamedRerouteUsage()) {attributeGetTypes [#1, #2]}
  n2 = Multiply(A=@n1.Roughness, B=ScalarParameter() {defaultValue=1, parameterName=Metal Roughness Multiplier})
  n3 = Saturate(None=Power(Base=@n1.Metallic))
  n4 = MaterialFunctionCall!(
   Mask=NamedRerouteUsage()
   Fresnel=NamedRerouteUsage()
   In=SetMaterialAttributes!(MaterialAttributes=@n1.MaterialAttributes, Roughness=LinearInterpolate(A=@n1.Roughness, B=@n2, Alpha=@n3)) {attributeSetTypes [#1]}
  ) {materialFunction=/Game/Global/MaterialFunctions/TechArt/MF_Advanced_CarClearCoat}
  n5 = MaterialFunctionCall() {materialFunction=&MF_CharacterParent_2023_PreFX}
  n6 = NamedRerouteDeclaration(None=@n5.Result) {name=PreFX_MA_Out, nodeColor=<rgba 1 0 0.9375 1>}
  n7 = MaterialFunctionCall!(In=@n6, Mask Input=@n5.M) {materialFunction=/Game/Global/Materials/Character_Parent_2023/Hair_Parent_2023/MaterialFunctions/MF_FN_Hair_2023}
  n8 = BlendMaterialAttributes(
   A=@n6
   B=StaticSwitchParameter(True=@n7.Result, False=@n6) {defaultValue=true, group=Skin FX, parameterName=UseFNHair2_0}
   Alpha=@n7.HairMask
  )
  n9 = StaticSwitchParameter(
   True=StaticSwitchParameter(
    True=MaterialFunctionCall!(MaterialInput=@n8) {materialFunction=/Game/Global/Materials/Character_Parent_2023/Hair_Parent_2023/MaterialFunctions/MF_FlyAwayHair_2023}.Result
    False=@n8
   ) {group=Skin FX, parameterName=IsFlyAwayHair?}
   False=@n6
  ) {group=Skin FX, parameterName=IsHair}
  n10 = StaticSwitchParameter(True=@n4.Result, False=@n9) {group=Skin FX, parameterName=UseAdvancedClearCoat}
  n11 = GetMaterialAttributes(None=@n10) {attributeGetTypes [#3, #4]}
  n12 = StaticSwitchParameter(True=@n11.Base Color, False=Desaturation(None=@n11.Base Color, Fraction=Constant() {r=1})) {group=Skin FX, parameterName="IsHair(SL)"}
  n13 = NamedRerouteUsage()
  NamedRerouteDeclaration(None=SetMaterialAttributes!(MaterialAttributes=@n11.MaterialAttributes, Base Color=LinearInterpolate(A=@n12, B=@n11.Base Color, Alpha=@n13), Emissive Color) {attributeSetTypes [#3, #4]}) {name=PreSL_MA, nodeColor=<rgba 0 0.6016 1 1>}
  NamedRerouteDeclaration(None=@n5.Tech Art Mask) {name=TechArtMask, nodeColor=<rgba 1 0.4219 0 1>}
  NamedRerouteDeclaration(None=ComponentMask(None=@n5.M) {b=true}) {name=SkinMask, nodeColor=<rgba 0.6406 0 1 1>}
  NamedRerouteDeclaration(None=@n10) {name=PreSL_Output, nodeColor=<rgba 0 1 0.4609 1>}
  n14 = NamedRerouteUsage()
  n15 = MaterialFunctionCall!(In=MaterialFunctionCall(In=NamedRerouteUsage()) {materialFunction=&/MF_S34SL_T2}.Result) {materialFunction=&MF_S34SL_T2}
  n16 = StaticSwitchParameter(
   True=MaterialFunctionCall!(In=NamedRerouteUsage()) {materialFunction=&/MF_S34SL_T3}.Result
   False=NamedRerouteUsage()
  ) {group=00_SuperLevels, parameterName=SL_Tier3}
  n17 = MaterialFunctionCall(
   Material=StaticSwitchParameter(
    True=MaterialFunctionCall!(In=@n14) {materialFunction=&/MF_S34SL_T1b}.Result
    False=StaticSwitchParameter(True=@n15.Result, False=@n16) {group=00_SuperLevels, parameterName=SL_Tier2}
   ) {group=00_SuperLevels, parameterName=SL_Tier1}
   M=@n5.M
   Pre Skinned Local Position=@n5.M
   Distance Blend=@n5.M
   Fresnel=NamedRerouteDeclaration(None=@n5.Fresnel) {name=Fresnel, nodeColor=<rgba 0 1 0.7422 1>}
  ) {materialFunction=&MF_CharacterParent_2023_PostFX}
  n18 = SetMaterialAttributes!(MaterialAttributes=@n17.Result, Normal=GetMaterialAttributes(None=@n9) {attributeGetTypes [#5]}.Normal) {attributeSetTypes [#5]}
  StaticSwitchParameter(
   True=MaterialFunctionCall(BaseNormal=@n4.NormalWithOrangePeel, AdditionalNormal=@n4.NormalWithOrangePeel) {materialFunction=&BlendAngleCorrectedNormals}.Result
   False=GetMaterialAttributes(None=StaticSwitchParameter(True=@n18, False=@n17.Result) {group=Skin FX, parameterName=UseAdvancedClearCoat}) {attributeGetTypes [#6]}.ClearCoatBottomNormal
  ) {group=Skin FX, parameterName=UseAdvancedClearCoat}
 ]
}
