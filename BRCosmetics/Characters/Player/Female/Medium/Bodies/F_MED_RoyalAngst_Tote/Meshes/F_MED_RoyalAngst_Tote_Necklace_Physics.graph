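# Physics asset for the dangling necklace on the female medium RoyalAngst Tote character cosmetic.
# Four capsule bodies form a chain from base_necklace, each link with narrow swing limits and a zero-degree twist limit.
# Capsules on the upper spine, both upper arms, right pectoral and front strap approximate the surrounding torso, apparently as collision for the chain.
# These body parts and the necklace base use free angular constraints to the spine.
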PhysicsAsset /BRCosmetics/Characters/Player/Female/Medium/Bodies/F_MED_RoyalAngst_Tote/Meshes/F_MED_RoyalAngst_Tote_Necklace_Physics
bodies {
 base_frontStrap [{shapeType=Capsule, length=0.505, radius=0.505, shapeName=base_frontStrap_capsule, rotation=<rot 0 -90 0>}]
 base_necklace [{shapeType=Capsule, length=0, radius=0.5, shapeName=base_necklace_capsule}]
 deform_pec_aim_r [{
   shapeType=Capsule
   length=3.0639
   radius=1.6633
   shapeName=deform_pec_aim_r_capsule
   center=<-1.0398 -3.2555 0.5653>
   rotation=<rot 5.1681 -73.1087 73.4772>
  }]
 dyn_necklace_01 [{shapeType=Capsule, length=2, radius=1, shapeName=dyn_necklace_01_capsule, center=<2 0 0>, rotation=<rot 90 0 0>}]
 dyn_necklace_02 [{shapeType=Capsule, length=2, radius=1, shapeName=dyn_necklace_02_capsule, center=<2 0 0>, rotation=<rot 90 0 0>}]
 dyn_necklace_03 [{shapeType=Capsule, length=0.75, radius=1.1, shapeName=dyn_necklace_03_capsule, center=<2 0 0>, rotation=<rot 90 0 0>}]
 dyn_necklace_04 [{shapeType=Capsule, length=1.5, radius=1.1, shapeName=dyn_necklace_03_capsule, center=<2 0 0>, rotation=<rot 90 0 0>}]
 spine_04 [
  {shapeType=Capsule, length=7.1143, radius=8.1275, shapeName=spine_05_capsule, center=<0.7463 3.9152 0>}
  {shapeType=Capsule, length=3.8048, radius=3.7602, shapeName=spine_04_capsule, center=<2.6439 8.2139 0>}
 ]
 spine_05 [
  {shapeType=Capsule, length=10.251, radius=8, shapeName=spine_05_capsule, center=<-0.2172 -0.4904 0>}
  {shapeType=Capsule, length=14.2893, radius=4.9698, shapeName=spine_05_capsule, center=<-1.6093 4.1681 0>, rotation=<rot 0 -0.1809 0>}
 ]
 upperarm_l [{shapeType=Capsule, length=25, radius=4, shapeName=upperarm_l_capsule, center=<14 0 0>, rotation=<rot 90 0 0>}]
 upperarm_r [{shapeType=Capsule, length=25, radius=4, shapeName=upperarm_r_capsule, center=<-14 0 0>, rotation=<rot 90 0 0>}]
}
constraints [
 {
  bone1Name=base_necklace
  bone2Name=spine_05
  swing1LimitDegrees=45
  swing1Motion=Free
  swing2LimitDegrees=45
  swing2Motion=Free
  twistLimitDegrees=45
  twistMotion=Free
 }
 {bone1Name=dyn_necklace_01, bone2Name=base_necklace, swing1LimitDegrees=5, swing2LimitDegrees=20, twistLimitDegrees=0}
 {bone1Name=dyn_necklace_02, bone2Name=dyn_necklace_01, swing1LimitDegrees=10, swing2LimitDegrees=25, twistLimitDegrees=0}
 {bone1Name=dyn_necklace_03, bone2Name=dyn_necklace_02, swing1LimitDegrees=15, swing2LimitDegrees=30, twistLimitDegrees=0}
 {
  bone1Name=upperarm_r
  bone2Name=spine_05
  swing1LimitDegrees=45
  swing1Motion=Free
  swing2LimitDegrees=45
  swing2Motion=Free
  twistLimitDegrees=45
  twistMotion=Free
 }
 {
  bone1Name=upperarm_l
  bone2Name=spine_05
  swing1LimitDegrees=45
  swing1Motion=Free
  swing2LimitDegrees=45
  swing2Motion=Free
  twistLimitDegrees=45
  twistMotion=Free
 }
 {
  bone1Name=spine_05
  bone2Name=spine_04
  swing1LimitDegrees=45
  swing1Motion=Free
  swing2LimitDegrees=45
  swing2Motion=Free
  twistLimitDegrees=45
  twistMotion=Free
 }
 {bone1Name=dyn_necklace_04, bone2Name=dyn_necklace_03, swing1LimitDegrees=8, swing2LimitDegrees=15, twistLimitDegrees=0}
 {
  bone1Name=deform_pec_aim_r
  bone2Name=spine_04
  swing1LimitDegrees=45
  swing1Motion=Free
  swing2LimitDegrees=45
  swing2Motion=Free
  twistLimitDegrees=45
  twistMotion=Free
 }
 {
  bone1Name=base_frontStrap
  bone2Name=spine_04
  swing1LimitDegrees=45
  swing1Motion=Free
  swing2LimitDegrees=45
  swing2Motion=Free
  twistLimitDegrees=45
  twistMotion=Free
 }
]
