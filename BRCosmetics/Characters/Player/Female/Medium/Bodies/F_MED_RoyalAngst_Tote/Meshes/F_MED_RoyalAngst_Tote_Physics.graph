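# Physics asset for the F_MED_RoyalAngst_Tote female player body, set up for secondary motion on a dangling grenade and a left pouch chain.
# Capsules on the pelvis, spine and thighs form a rough body volume, linked by free angular constraints.
# The three-link pouch chain hangs from a base locked to spine_03, while the grenade's base attaches freely to spine_01.
# Both dangling parts have asymmetric swing limits and zero-degree twist limits.
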PhysicsAsset /BRCosmetics/Characters/Player/Female/Medium/Bodies/F_MED_RoyalAngst_Tote/Meshes/F_MED_RoyalAngst_Tote_Physics
bodies {
 base_grenade [{shapeType=Capsule, length=0, radius=0.5, shapeName=base_pouch_capsule}]
 base_pouch [{shapeType=Capsule, length=0, radius=0.5, shapeName=base_pouch_capsule}]
 dyn_PouchChain_L_01 [{shapeType=Capsule, length=1.0307, radius=1, shapeName=dyn_PouchChain_L_01_capsule, center=<2.6245 0 -0.0481>, rotation=<rot 86.1347 0 0>}]
 dyn_PouchChain_L_02 [{shapeType=Capsule, length=1.2294, radius=1.11, shapeName=dyn_PouchChain_L_02_capsule, center=<1.7153 0 -0.3454>, rotation=<rot 85.6636 0 0>}]
 dyn_PouchChain_L_03 [{shapeType=Capsule, length=0.6501, radius=1.0748, shapeName=dyn_PouchChain_L_03_capsule, center=<1.0933 0 -0.5465>, rotation=<rot 90 0 0>}]
 dyn_grenade [{
   shapeType=Capsule
   length=8.8762
   radius=2.5
   shapeName=dyn_grenade_capsule
   center=<3.0476 0.5302 0.503>
   rotation=<rot 84.6287 -172.5197 179.9998>
  }]
 pelvis [{shapeType=Capsule, length=8.5097, radius=9.0212, shapeName=pelvis_capsule, center=<3.3274 -1.6651 0>}]
 spine_01 [{shapeType=Capsule, length=2, radius=1, shapeName=spine_01_capsule}]
 spine_02 [{
   shapeType=Capsule
   length=7.0751
   radius=6.6547
   shapeName=spine_05_capsule
   center=<-0.7543 1.009 -2.7131>
   rotation=<rot 0 -0.0012 29.6399>
  }]
 spine_03 [{
   shapeType=Capsule
   length=7.0751
   radius=6.6547
   shapeName=spine_05_capsule
   center=<-0.7543 1.009 -2.7131>
   rotation=<rot 0 -0.0012 29.6399>
  }]
 thigh_r [{shapeType=Capsule, length=0.5437, radius=6.4563, shapeName=thigh_l_capsule, center=<-0.4092 0 -0.3557>, rotation=<rot 90 0 180>}]
 thigh_twist_01_l [{shapeType=Capsule, length=20, radius=8, shapeName=thigh_l_capsule, center=<13.2033 0.4632 0>, rotation=<rot 90 87.3911 88.7771>}]
 thigh_twist_01_r [{shapeType=Capsule, length=20, radius=8, shapeName=thigh_l_capsule, center=<-13.2033 0 0>, rotation=<rot 90 0 180>}]
}
constraints [
 {
  bone1Name=spine_01
  bone2Name=pelvis
  swing1LimitDegrees=45
  swing1Motion=Free
  swing2LimitDegrees=45
  swing2Motion=Free
  twistLimitDegrees=45
  twistMotion=Free
 }
 {
  bone1Name=spine_02
  bone2Name=spine_01
  swing1LimitDegrees=45
  swing1Motion=Free
  swing2LimitDegrees=45
  swing2Motion=Free
  twistLimitDegrees=45
  twistMotion=Free
 }
 {bone1Name=dyn_grenade, bone2Name=base_grenade, swing1LimitDegrees=40, swing2LimitDegrees=65, twistLimitDegrees=0}
 {
  bone1Name=base_pouch
  bone2Name=spine_03
  swing1LimitDegrees=45
  swing1Motion=Locked
  swing2LimitDegrees=45
  swing2Motion=Locked
  twistLimitDegrees=45
  twistMotion=Locked
 }
 {bone1Name=dyn_PouchChain_L_01, bone2Name=base_pouch, swing1LimitDegrees=15, swing2LimitDegrees=35, twistLimitDegrees=0}
 {bone1Name=dyn_PouchChain_L_02, bone2Name=dyn_PouchChain_L_01, swing1LimitDegrees=25, swing2LimitDegrees=15, twistLimitDegrees=0}
 {bone1Name=dyn_PouchChain_L_03, bone2Name=dyn_PouchChain_L_02, swing1LimitDegrees=25, swing2LimitDegrees=15, twistLimitDegrees=0}
 {
  bone1Name=spine_03
  bone2Name=spine_02
  swing1LimitDegrees=45
  swing1Motion=Free
  swing2LimitDegrees=45
  swing2Motion=Free
  twistLimitDegrees=45
  twistMotion=Free
 }
 {
  bone1Name=thigh_twist_01_l
  bone2Name=pelvis
  swing1LimitDegrees=45
  swing1Motion=Free
  swing2LimitDegrees=45
  swing2Motion=Free
  twistLimitDegrees=45
  twistMotion=Free
 }
 {
  bone1Name=thigh_r
  bone2Name=pelvis
  swing1LimitDegrees=45
  swing1Motion=Free
  swing2LimitDegrees=45
  swing2Motion=Free
  twistLimitDegrees=45
  twistMotion=Free
 }
 {
  bone1Name=thigh_twist_01_r
  bone2Name=thigh_r
  swing1LimitDegrees=45
  swing1Motion=Free
  swing2LimitDegrees=45
  swing2Motion=Free
  twistLimitDegrees=45
  twistMotion=Free
 }
 {
  bone1Name=base_grenade
  bone2Name=spine_01
  swing1LimitDegrees=45
  swing1Motion=Free
  swing2LimitDegrees=45
  swing2Motion=Free
  twistLimitDegrees=45
  twistMotion=Free
 }
]
