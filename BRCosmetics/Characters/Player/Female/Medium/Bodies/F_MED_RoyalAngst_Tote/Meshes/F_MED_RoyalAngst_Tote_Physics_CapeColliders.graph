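# Physics asset of cape colliders for the F_MED_RoyalAngst_Tote female medium body cosmetic, apparently serving as collision shapes for cape simulation.
# Each covered bone carries one capsule, spanning pelvis, upper spine chain, clavicles, arms to the forearms, and legs to the feet, with no head, neck or hand bodies.
# Constraints chain the bodies out from the pelvis with all swing and twist motion free, so they impose no angular limits.
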PhysicsAsset /BRCosmetics/Characters/Player/Female/Medium/Bodies/F_MED_RoyalAngst_Tote/Meshes/F_MED_RoyalAngst_Tote_Physics_CapeColliders
bodies {
 calf_l [{shapeType=Capsule, length=43.5, radius=12, shapeName=calf_l_capsule, center=<-21.93 -1.9975 0>, rotation=<rot 90 0 0>}]
 calf_r [{shapeType=Capsule, length=43.5, radius=12, shapeName=calf_r_capsule, center=<21.93 1.9975 0>, rotation=<rot 90 0 180>}]
 clavicle_l [{shapeType=Capsule, length=9, radius=7, shapeName=clavicle_l_capsule, center=<3.0813 -0.5574 0>, rotation=<rot 90 0 0>}]
 clavicle_r [{shapeType=Capsule, length=9, radius=7, shapeName=clavicle_r_capsule, center=<-3.0813 0.5574 0>, rotation=<rot 90 0 180>}]
 foot_l [{shapeType=Capsule, length=15.5, radius=7, shapeName=foot_l_capsule, center=<-4.1312 6.0495 0>, rotation=<rot 90 -165.9638 104.0362>}]
 foot_r [{shapeType=Capsule, length=15.5, radius=7, shapeName=foot_r_capsule, center=<4.1312 -6.0495 0>, rotation=<rot 90 180 -90>}]
 lowerarm_l [{shapeType=Capsule, length=17, radius=6, shapeName=lowerarm_l_capsule, center=<7.8177 0 0>, rotation=<rot 90 0 0>}]
 lowerarm_r [{shapeType=Capsule, length=17, radius=6, shapeName=lowerarm_r_capsule, center=<-7.8177 0 0>, rotation=<rot 90 0 180>}]
 pelvis [{shapeType=Capsule, length=8.5, radius=19, shapeName=pelvis_capsule, center=<-4.7902 3.4411 0>}]
 spine_02 [{shapeType=Capsule, length=9, radius=16, shapeName=spine_02_capsule}]
 spine_03 [{shapeType=Capsule, length=8.5, radius=16, shapeName=spine_03_capsule}]
 spine_04 [{shapeType=Capsule, length=6.5, radius=17, shapeName=spine_04_capsule}]
 spine_05 [{shapeType=Capsule, length=11.5, radius=16, shapeName=spine_05_capsule, center=<5.027 0 0>}]
 thigh_l [{shapeType=Capsule, length=39, radius=15, shapeName=thigh_l_capsule, center=<-22.1632 1.1016 0>, rotation=<rot 90 180 -176.3585>}]
 thigh_r [{shapeType=Capsule, length=39, radius=15, shapeName=thigh_r_capsule, center=<22.1632 -1.1016 0>, rotation=<rot 90 180 3.6415>}]
 upperarm_l [{shapeType=Capsule, length=20.5, radius=6, shapeName=upperarm_l_capsule, center=<13.1902 0 0>, rotation=<rot 90 0 0>}]
 upperarm_r [{shapeType=Capsule, length=20.5, radius=6, shapeName=upperarm_r_capsule, center=<-13.1902 0 0>, rotation=<rot 90 0 180>}]
}
constraints [
 {
  bone1Name=spine_03
  bone2Name=spine_02
  swing1LimitDegrees=45
  swing1Motion=Free
  swing2LimitDegrees=45
  swing2Motion=Free
  twistLimitDegrees=45
  twistMotion=Free
 }
 {
  bone1Name=spine_04
  bone2Name=spine_03
  swing1LimitDegrees=45
  swing1Motion=Free
  swing2LimitDegrees=45
  swing2Motion=Free
  twistLimitDegrees=45
  twistMotion=Free
 }
 {
  bone1Name=spine_05
  bone2Name=spine_04
  swing1LimitDegrees=45
  swing1Motion=Free
  swing2LimitDegrees=45
  swing2Motion=Free
  twistLimitDegrees=45
  twistMotion=Free
 }
 {
  bone1Name=upperarm_l
  bone2Name=clavicle_l
  swing1LimitDegrees=45
  swing1Motion=Free
  swing2LimitDegrees=45
  swing2Motion=Free
  twistLimitDegrees=45
  twistMotion=Free
 }
 {
  bone1Name=lowerarm_l
  bone2Name=upperarm_l
  swing1LimitDegrees=45
  swing1Motion=Free
  swing2LimitDegrees=45
  swing2Motion=Free
  twistLimitDegrees=45
  twistMotion=Free
 }
 {
  bone1Name=clavicle_l
  bone2Name=spine_05
  swing1LimitDegrees=45
  swing1Motion=Free
  swing2LimitDegrees=45
  swing2Motion=Free
  twistLimitDegrees=45
  twistMotion=Free
 }
 {
  bone1Name=clavicle_r
  bone2Name=spine_05
  swing1LimitDegrees=45
  swing1Motion=Free
  swing2LimitDegrees=45
  swing2Motion=Free
  twistLimitDegrees=45
  twistMotion=Free
 }
 {
  bone1Name=upperarm_r
  bone2Name=clavicle_r
  swing1LimitDegrees=45
  swing1Motion=Free
  swing2LimitDegrees=45
  swing2Motion=Free
  twistLimitDegrees=45
  twistMotion=Free
 }
 {
  bone1Name=lowerarm_r
  bone2Name=upperarm_r
  swing1LimitDegrees=45
  swing1Motion=Free
  swing2LimitDegrees=45
  swing2Motion=Free
  twistLimitDegrees=45
  twistMotion=Free
 }
 {
  bone1Name=calf_l
  bone2Name=thigh_l
  swing1LimitDegrees=45
  swing1Motion=Free
  swing2LimitDegrees=45
  swing2Motion=Free
  twistLimitDegrees=45
  twistMotion=Free
 }
 {
  bone1Name=foot_l
  bone2Name=calf_l
  swing1LimitDegrees=45
  swing1Motion=Free
  swing2LimitDegrees=45
  swing2Motion=Free
  twistLimitDegrees=45
  twistMotion=Free
 }
 {
  bone1Name=calf_r
  bone2Name=thigh_r
  swing1LimitDegrees=45
  swing1Motion=Free
  swing2LimitDegrees=45
  swing2Motion=Free
  twistLimitDegrees=45
  twistMotion=Free
 }
 {
  bone1Name=foot_r
  bone2Name=calf_r
  swing1LimitDegrees=45
  swing1Motion=Free
  swing2LimitDegrees=45
  swing2Motion=Free
  twistLimitDegrees=45
  twistMotion=Free
 }
 {
  bone1Name=spine_02
  bone2Name=pelvis
  swing1LimitDegrees=45
  swing1Motion=Free
  swing2LimitDegrees=45
  swing2Motion=Free
  twistLimitDegrees=45
  twistMotion=Free
 }
 {
  bone1Name=thigh_l
  bone2Name=pelvis
  swing1LimitDegrees=45
  swing1Motion=Free
  swing2LimitDegrees=45
  swing2Motion=Free
  twistLimitDegrees=45
  twistMotion=Free
 }
 {
  bone1Name=thigh_r
  bone2Name=pelvis
  swing1LimitDegrees=45
  swing1Motion=Free
  swing2LimitDegrees=45
  swing2Motion=Free
  twistLimitDegrees=45
  twistMotion=Free
 }
]
