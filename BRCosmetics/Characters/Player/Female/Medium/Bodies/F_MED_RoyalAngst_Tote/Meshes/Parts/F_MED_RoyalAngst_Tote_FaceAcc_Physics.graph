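# Physics asset for the RoyalAngst Tote female face accessory mesh, setting up capsule bodies and constraints for dynamic hair.
# Left, middle and right six-bone chains have tight swing and twist limits, rooted at the head, the right through base_hair_R.
# Angularly free constraints tie the left and right tips to the middle tip, linking the strands.
# Capsules on head, pelvis, spine, clavicles and upper arms cover the body, with arm joints angularly free.
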PhysicsAsset /BRCosmetics/Characters/Player/Female/Medium/Bodies/F_MED_RoyalAngst_Tote/Meshes/Parts/F_MED_RoyalAngst_Tote_FaceAcc_Physics
bodies {
 base_hair_R [{shapeType=Capsule, length=6.4792, radius=4, shapeName=dyn_hair_R_01_capsule, center=<7.5687 0 -3.3707>, rotation=<rot 80 180 180>}]
 clavicle_l [{shapeType=Capsule, length=8.3419, radius=5.7, shapeName=clavicle_l_capsule, center=<4.0541 -1.0758 0.0036>, rotation=<rot 90 0 0>}]
 clavicle_r [{shapeType=Capsule, length=8.3419, radius=5.7, shapeName=clavicle_r_capsule, center=<-4.0541 1.0758 -0.0036>, rotation=<rot 90 0 0>}]
 dyn_hair_L_01 [{shapeType=Capsule, length=4, radius=4, shapeName=dyn_hair_R_04_capsule, center=<2.3706 0 -2>, rotation=<rot 90 0 0>}]
 dyn_hair_L_02 [{shapeType=Capsule, length=6.5, radius=4.1, shapeName=dyn_hair_R_04_capsule, center=<1.3565 0 -1>, rotation=<rot 84.9633 170.1886 170.1056>}]
 dyn_hair_L_03 [{shapeType=Capsule, length=4, radius=3.6, shapeName=dyn_hair_L_04_capsule, center=<2.3721 1 0>, rotation=<rot 71.1033 175.0731 174.9942>}]
 dyn_hair_L_04 [{shapeType=Capsule, length=4, radius=3.6, shapeName=dyn_hair_L_04_capsule, center=<2.0103 2 0.6>, rotation=<rot 90 0 0>}]
 dyn_hair_L_05 [{shapeType=Capsule, length=4, radius=4, shapeName=dyn_hair_L_05_capsule, center=<2.7326 2.6 -1.3>, rotation=<rot 90 0 0>}]
 dyn_hair_L_06 [{shapeType=Capsule, length=5, radius=4.2, shapeName=dyn_hair_L_06_capsule, center=<3.0963 1 -0.5>, rotation=<rot 90 0 0>}]
 dyn_hair_M_01 [{shapeType=Capsule, length=4, radius=4, shapeName=dyn_hair_R_04_capsule, center=<2.3706 0 -2>, rotation=<rot 90 0 0>}]
 dyn_hair_M_02 [{shapeType=Capsule, length=6.5, radius=4.1, shapeName=dyn_hair_R_04_capsule, center=<1.3539 0 -1>, rotation=<rot 85.0001 180 180>}]
 dyn_hair_M_03 [{shapeType=Capsule, length=4, radius=3.6, shapeName=dyn_hair_M_04_capsule, center=<2.3707 0 0>, rotation=<rot 71.1031 180 180>}]
 dyn_hair_M_04 [{shapeType=Capsule, length=4, radius=4, shapeName=dyn_hair_M_04_capsule, center=<2.0103 0 0.6>, rotation=<rot 90 0 0>}]
 dyn_hair_M_05 [{shapeType=Capsule, length=4, radius=4.5, shapeName=dyn_hair_M_05_capsule, center=<2.6361 0 0>, rotation=<rot 90 0 0>}]
 dyn_hair_M_06 [{shapeType=Capsule, length=5, radius=4.2, shapeName=dyn_hair_M_06_capsule, center=<3.1323 0 0>, rotation=<rot 90 0 0>}]
 dyn_hair_R_01 [{shapeType=Capsule, length=6, radius=4.5, shapeName=dyn_hair_R_01_capsule, center=<4 0 -4>, rotation=<rot 90 0 0>}]
 dyn_hair_R_02 [{
   shapeType=Capsule
   length=6.5
   radius=4.1
   shapeName=dyn_hair_R_04_capsule
   center=<2.1027 -1 -2.6>
   rotation=<rot 65.0239 -178.1894 -177.9525>
  }]
 dyn_hair_R_03 [{shapeType=Capsule, length=4, radius=3.6, shapeName=dyn_hair_R_04_capsule, center=<2.3728 -1 0>, rotation=<rot 71.1033 -175.0731 -174.9942>}]
 dyn_hair_R_04 [{shapeType=Capsule, length=4, radius=3.6, shapeName=dyn_hair_R_04_capsule, center=<2.0103 -2 0.6>, rotation=<rot 90 0 0>}]
 dyn_hair_R_05 [{shapeType=Capsule, length=4, radius=4, shapeName=dyn_hair_R_05_capsule, center=<2.5947 -2.6 -1.3>, rotation=<rot 90 0 0>}]
 dyn_hair_R_06 [{shapeType=Capsule, length=5, radius=4.2, shapeName=dyn_hair_R_06_capsule, center=<3.1094 -1 -0.5>, rotation=<rot 90 0 0>}]
 head [{shapeType=Capsule, length=1, radius=1, shapeName=head_capsule}]
 pelvis [{shapeType=Capsule, length=6.6871, radius=12.2436, shapeName=pelvis_capsule}]
 spine_01 [{shapeType=Capsule, length=8.5167, radius=10, shapeName=spine_01_capsule, center=<0.4244 0.6535 0>}]
 spine_02 [{shapeType=Capsule, length=4.7259, radius=13, shapeName=spine_02_capsule, center=<0 5.2589 0>}]
 spine_03 [{shapeType=Capsule, length=4.7259, radius=13, shapeName=spine_03_capsule, center=<0 5 0>}]
 spine_04 [{shapeType=Capsule, length=60, radius=35, shapeName=spine_04_capsule, center=<-1.8252 28.9226 0>}]
 spine_05 [{shapeType=Capsule, length=10.1632, radius=8.6265, shapeName=spine_05_capsule, center=<2.5194 0.7308 0>}]
 upperarm_l [{shapeType=Capsule, length=20, radius=13.5, shapeName=upperarm_r_capsule, center=<25 6 2>, rotation=<rot 90 0 0>}]
 upperarm_r [{shapeType=Capsule, length=20, radius=13.5, shapeName=upperarm_r_capsule, center=<-25 -6 -2>, rotation=<rot 90 0 0>}]
 upperarm_twist_01_l [{shapeType=Capsule, length=40, radius=10, shapeName=upperarm_l_capsule, center=<18 2 2>, rotation=<rot 90 0 0>}]
 upperarm_twist_01_r [{shapeType=Capsule, length=40, radius=10, shapeName=upperarm_r_capsule, center=<-18 -2 -2>, rotation=<rot 90 0 0>}]
}
constraints [
 {bone1Name=dyn_hair_L_06, bone2Name=dyn_hair_L_05, swing1LimitDegrees=10, swing2LimitDegrees=35, twistLimitDegrees=5}
 {bone1Name=dyn_hair_L_05, bone2Name=dyn_hair_L_04, swing1LimitDegrees=10, swing2LimitDegrees=35, twistLimitDegrees=5}
 {bone1Name=dyn_hair_L_04, bone2Name=dyn_hair_L_03, swing1LimitDegrees=10, swing2LimitDegrees=35, twistLimitDegrees=5}
 {bone1Name=dyn_hair_L_03, bone2Name=dyn_hair_L_02, swing1LimitDegrees=10, swing2LimitDegrees=35, twistLimitDegrees=5}
 {bone1Name=dyn_hair_M_06, bone2Name=dyn_hair_M_05, swing1LimitDegrees=10, swing2LimitDegrees=35, twistLimitDegrees=5}
 {bone1Name=dyn_hair_M_05, bone2Name=dyn_hair_M_04, swing1LimitDegrees=10, swing2LimitDegrees=35, twistLimitDegrees=5}
 {bone1Name=dyn_hair_M_04, bone2Name=dyn_hair_M_03, swing1LimitDegrees=10, swing2LimitDegrees=35, twistLimitDegrees=5}
 {bone1Name=dyn_hair_M_03, bone2Name=dyn_hair_M_02, swing1LimitDegrees=10, swing2LimitDegrees=35, twistLimitDegrees=5}
 {bone1Name=dyn_hair_R_06, bone2Name=dyn_hair_R_05, swing1LimitDegrees=7, swing2LimitDegrees=35, twistLimitDegrees=10}
 {bone1Name=dyn_hair_R_05, bone2Name=dyn_hair_R_04, swing1LimitDegrees=7, swing2LimitDegrees=35, twistLimitDegrees=10}
 {bone1Name=dyn_hair_R_04, bone2Name=dyn_hair_R_03, swing1LimitDegrees=7, swing2LimitDegrees=35, twistLimitDegrees=10}
 {bone1Name=dyn_hair_R_03, bone2Name=dyn_hair_R_02, swing1LimitDegrees=7, swing2LimitDegrees=35, twistLimitDegrees=10}
 {bone1Name=dyn_hair_L_02, bone2Name=dyn_hair_L_01, swing1LimitDegrees=10, swing2LimitDegrees=35, twistLimitDegrees=3}
 {bone1Name=dyn_hair_M_02, bone2Name=dyn_hair_M_01, swing1LimitDegrees=10, swing2LimitDegrees=35, twistLimitDegrees=3}
 {bone1Name=dyn_hair_R_02, bone2Name=dyn_hair_R_01, swing1LimitDegrees=7, swing2LimitDegrees=35, twistLimitDegrees=10}
 {bone1Name=dyn_hair_M_01, bone2Name=head, swing1LimitDegrees=10, swing2LimitDegrees=35, twistLimitDegrees=3}
 {bone1Name=dyn_hair_R_01, bone2Name=base_hair_R, swing1LimitDegrees=7, swing2LimitDegrees=35, twistLimitDegrees=10}
 {bone1Name=dyn_hair_L_01, bone2Name=head, swing1LimitDegrees=10, swing2LimitDegrees=35, twistLimitDegrees=3}
 {
  bone1Name=dyn_hair_R_06
  bone2Name=dyn_hair_M_06
  swing1LimitDegrees=45
  swing1Motion=Free
  swing2LimitDegrees=45
  swing2Motion=Free
  twistLimitDegrees=45
  twistMotion=Free
 }
 {
  bone1Name=dyn_hair_L_06
  bone2Name=dyn_hair_M_06
  swing1LimitDegrees=45
  swing1Motion=Free
  swing2LimitDegrees=45
  swing2Motion=Free
  twistLimitDegrees=45
  twistMotion=Free
 }
 {
  bone1Name=upperarm_twist_01_r
  bone2Name=upperarm_r
  swing1LimitDegrees=45
  swing1Motion=Free
  swing2LimitDegrees=45
  swing2Motion=Free
  twistLimitDegrees=45
  twistMotion=Free
 }
 {
  bone1Name=upperarm_r
  bone2Name=clavicle_r
  swing1LimitDegrees=45
  swing1Motion=Free
  swing2LimitDegrees=45
  swing2Motion=Free
  twistLimitDegrees=45
  twistMotion=Free
 }
 {
  bone1Name=upperarm_twist_01_l
  bone2Name=upperarm_l
  swing1LimitDegrees=45
  swing1Motion=Free
  swing2LimitDegrees=45
  swing2Motion=Free
  twistLimitDegrees=45
  twistMotion=Free
 }
 {
  bone1Name=upperarm_l
  bone2Name=clavicle_l
  swing1LimitDegrees=45
  swing1Motion=Free
  swing2LimitDegrees=45
  swing2Motion=Free
  twistLimitDegrees=45
  twistMotion=Free
 }
 {bone1Name=base_hair_R, bone2Name=head, swing1LimitDegrees=7, swing2LimitDegrees=10, twistLimitDegrees=3}
]
